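# Physics asset for the cape of the medium female BinGrass character, with capsule bodies on the character's body and cape bones.
# Four six-bone cape chains hang from master bones on the clavicles and upper spine, using swing limits and zero twist limit, while body joints rotate freely.
# Cross-constraints tie lower outer and middle segments, free only on the right, and join the two middle chains.
# An oversized, unconstrained capsule sits on gravity_cloth.
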PhysicsAsset /BRCosmetics/Characters/Player/Female/Medium/Bodies/F_MED_BinGrass/Meshes/F_MED_BinGrass_Cape_Physics
bodies {
 calf_l [{shapeType=Capsule, length=34.8227, radius=15.5124, shapeName=calf_l_capsule, center=<-18.4993 4.0838 0>, rotation=<rot 90 0 0>}]
 calf_r [{shapeType=Capsule, length=34.8227, radius=15.5124, shapeName=calf_r_capsule, center=<18.4993 -1.6119 0>, rotation=<rot 90 0 0>}]
 cape_01_l [
  {
   shapeType=Capsule
   length=17.1835
   radius=6.8165
   shapeName=cape_02_r_capsule
   center=<10.0617 0.002 3.8696>
   rotation=<rot -0.0003 0 90.029>
  }
  {shapeType=Capsule, length=16.674, radius=7.326, shapeName=cape_02_r_capsule, center=<24.2904 0.0014 2.8263>, rotation=<rot -0.0003 0 90.029>}
 ]
 cape_01_r [
  {
   shapeType=Capsule
   length=16.711
   radius=7.289
   shapeName=cape_02_r_capsule
   center=<10.0616 0.0058 3.8697>
   rotation=<rot -0.0001 0 90.0853>
  }
  {
   shapeType=Capsule
   length=16.674
   radius=7.326
   shapeName=cape_02_r_capsule
   center=<24.2904 0.0042 2.8264>
   rotation=<rot -0.0001 0 90.0853>
  }
 ]
 cape_02_l [
  {shapeType=Capsule, length=16.197, radius=7.803, shapeName=cape_02_r_capsule, center=<10.0617 0.002 3.8696>, rotation=<rot -0.0003 0 90.029>}
  {
   shapeType=Capsule
   length=16.0263
   radius=7.9737
   shapeName=cape_02_r_capsule
   center=<24.2904 0.0022 4.2925>
   rotation=<rot -0.0003 0 90.029>
  }
 ]
 cape_02_r [
  {
   shapeType=Capsule
   length=15.7244
   radius=8.2756
   shapeName=cape_02_r_capsule
   center=<10.0616 0.0058 3.8697>
   rotation=<rot -0.0001 0 90.0853>
  }
  {
   shapeType=Capsule
   length=16.0263
   radius=7.9737
   shapeName=cape_02_r_capsule
   center=<24.2904 0.0064 4.2926>
   rotation=<rot -0.0001 0 90.0853>
  }
 ]
 cape_03_l [{
   shapeType=Capsule
   length=10.9492
   radius=10.5675
   shapeName=cape_03_l_capsule
   center=<9.441 0.0151 7.0375>
   rotation=<rot -0.0002 0 90.1232>
  }]
 cape_03_r [{
   shapeType=Capsule
   length=10.9492
   radius=10.5675
   shapeName=cape_03_l_capsule
   center=<9.441 0.0151 7.0375>
   rotation=<rot -0.0002 0 90.1232>
  }]
 cape_04_l [{
   shapeType=Capsule
   length=5.8818
   radius=12.7927
   shapeName=cape_04_l_capsule
   center=<9.441 0.0151 7.0375>
   rotation=<rot -0.0002 0 90.1232>
  }]
 cape_04_r [{shapeType=Capsule, length=5.882, radius=12.7927, shapeName=cape_04_l_capsule, center=<9.441 0.0151 7.0375>, rotation=<rot -0.0002 0 90.1232>}]
 cape_05_l [{shapeType=Capsule, length=3.762, radius=13.5673, shapeName=cape_05_l_capsule, center=<9.441 0.0151 7.0375>, rotation=<rot -0.0002 0 90.1232>}]
 cape_05_r [{
   shapeType=Capsule
   length=3.7622
   radius=13.5673
   shapeName=cape_05_l_capsule
   center=<9.441 0.0151 7.0375>
   rotation=<rot -0.0002 0 90.1232>
  }]
 cape_06_l [{
   shapeType=Capsule
   length=6.2582
   radius=13.4686
   shapeName=cape_06_l_capsule
   center=<9.441 0.0151 7.0375>
   rotation=<rot -0.0002 0 90.1232>
  }]
 cape_06_r [{
   shapeType=Capsule
   length=6.2582
   radius=13.4686
   shapeName=cape_06_l_capsule
   center=<9.441 0.0151 7.0375>
   rotation=<rot -0.0002 0 90.1232>
  }]
 cape_l_MSTR [{shapeType=Capsule, length=1, radius=1, shapeName=cape_l_MSTR_capsule, rotation=<rot 0 -90 0>}]
 cape_mid_01_l [{shapeType=Capsule, length=8.0095, radius=11.0887, shapeName=cape_mid_01_l_capsule, center=<13.4315 0 10.4878>, rotation=<rot 77.823 0 180>}]
 cape_mid_01_r [{
   shapeType=Capsule
   length=8.0095
   radius=11.0887
   shapeName=cape_mid_01_r_capsule
   center=<13.4315 0 10.4878>
   rotation=<rot 77.823 0.0001 -179.9999>
  }]
 cape_mid_02_l [{
   shapeType=Capsule
   length=12.2343
   radius=8.5883
   shapeName=cape_mid_02_l_capsule
   center=<17.4918 -0.1391 5.8682>
   rotation=<rot 90 -172.875 -81.5977>
  }]
 cape_mid_02_r [{
   shapeType=Capsule
   length=12.2343
   radius=8.5883
   shapeName=cape_mid_02_r_capsule
   center=<17.4918 -0.1391 5.8682>
   rotation=<rot 90 -175.2364 -83.959>
  }]
 cape_mid_03_l [{
   shapeType=Capsule
   length=7.9685
   radius=11.0738
   shapeName=cape_mid_03_l_capsule
   center=<9.441 0.0151 7.0375>
   rotation=<rot -0.0002 0 90.1232>
  }]
 cape_mid_03_r [{
   shapeType=Capsule
   length=7.9685
   radius=11.0738
   shapeName=cape_mid_03_r_capsule
   center=<9.441 0.0151 7.0375>
   rotation=<rot -0.0002 0 90.1232>
  }]
 cape_mid_04_l [{shapeType=Capsule, length=4.392, radius=12.6605, shapeName=cape_04_l_capsule, center=<9.441 0.0193 8.9797>, rotation=<rot -0.0002 0 90.1232>}]
 cape_mid_04_r [{shapeType=Capsule, length=4.392, radius=12.6605, shapeName=cape_04_l_capsule, center=<9.441 0.0193 8.9797>, rotation=<rot -0.0002 0 90.1232>}]
 cape_mid_05_l [{
   shapeType=Capsule
   length=10.2832
   radius=13.1817
   shapeName=cape_mid_05_l_capsule
   center=<9.441 0.0199 9.2777>
   rotation=<rot -0.0002 0 90.1232>
  }]
 cape_mid_05_r [{
   shapeType=Capsule
   length=9.7906
   radius=13.6742
   shapeName=cape_mid_05_r_capsule
   center=<9.441 0.0199 9.2777>
   rotation=<rot -0.0002 0 90.1232>
  }]
 cape_mid_06_l [{
   shapeType=Capsule
   length=15.4991
   radius=13.6418
   shapeName=cape_mid_06_l_capsule
   center=<9.441 0.0151 7.0375>
   rotation=<rot -0.0002 0 90.1232>
  }]
 cape_mid_06_r [{
   shapeType=Capsule
   length=15.5122
   radius=13.6286
   shapeName=cape_mid_06_r_capsule
   center=<9.441 0.0151 7.0375>
   rotation=<rot -0.0002 0 90.1232>
  }]
 cape_mid_l_MSTR [{shapeType=Capsule, length=1, radius=1, shapeName=cape_mid_l_MSTR_capsule, rotation=<rot 0 -90 0>}]
 cape_mid_r_MSTR [{shapeType=Capsule, length=1, radius=1, shapeName=cape_mid_r_MSTR_capsule, rotation=<rot 0 -90 0>}]
 cape_r_MSTR [{shapeType=Capsule, length=1, radius=1, shapeName=cape_r_MSTR_capsule, rotation=<rot 0 -90 0>}]
 clavicle_l [{shapeType=Capsule, length=12.0737, radius=7.5263, shapeName=clavicle_l_capsule, center=<9.3 0 0>, rotation=<rot 81 0 0>}]
 clavicle_r [{shapeType=Capsule, length=12.0737, radius=7.5263, shapeName=clavicle_r_capsule, center=<-9.3 0 0>, rotation=<rot 81 0 0>}]
 foot_l [{shapeType=Capsule, length=16.1864, radius=13.9127, shapeName=foot_l_capsule, center=<-4.0851 7.4016 0>, rotation=<rot 90 180 -90>}]
 foot_r [{shapeType=Capsule, length=14.746, radius=13.9127, shapeName=foot_r_capsule, center=<5.1029 -6.88 0>, rotation=<rot 90 180 -90>}]
 gravity_cloth [{
   shapeType=Capsule
   length=106.4505
   radius=152.3466
   shapeName=gravity_cloth_capsule
   center=<0.0034 -38.5109 -145.4722>
   rotation=<rot -0.0003 -90 90.1582>
  }]
 lowerarm_l [{shapeType=Capsule, length=30.0644, radius=8.2153, shapeName=lowerarm_r_capsule, center=<18.3461 1.1426 0>, rotation=<rot 90 0 0>}]
 lowerarm_r [{shapeType=Capsule, length=28.148, radius=10.1317, shapeName=lowerarm_r_capsule, center=<-15.1494 -3.3146 0>, rotation=<rot 90 0 0>}]
 pelvis [{shapeType=Capsule, length=13.7558, radius=12.8138, shapeName=pelvis_capsule, center=<0 -3.1413 0>}]
 spine_01 [{shapeType=Capsule, length=20.4523, radius=6.1161, shapeName=pelvis_capsule, center=<-0.4465 -13.5276 0>}]
 spine_02 [{shapeType=Capsule, length=8.1673, radius=16.1924, shapeName=spine_02_capsule}]
 spine_03 [{shapeType=Capsule, length=10.1367, radius=14.7438, shapeName=spine_03_capsule, center=<0 0.7308 0>}]
 spine_04 [{shapeType=Capsule, length=20.1066, radius=13.2328, shapeName=spine_04_capsule, center=<0 2.4161 0>}]
 spine_05 [{shapeType=Capsule, length=28.2914, radius=10.5568, shapeName=spine_05_capsule, center=<0 1.6185 0>}]
 thigh_l [{shapeType=Capsule, length=26.6523, radius=14.6065, shapeName=thigh_l_capsule, center=<-18.4993 2.0351 0>, rotation=<rot 90 0 0>}]
 thigh_r [{shapeType=Capsule, length=20.4323, radius=20.8265, shapeName=thigh_r_capsule, center=<18.4993 -6.9017 0>, rotation=<rot 90 0 0>}]
 upperarm_l [{shapeType=Capsule, length=14.3368, radius=12.7017, shapeName=upperarm_l_capsule, center=<17.4821 0.5142 0>, rotation=<rot 90 0 0>}]
 upperarm_r [{shapeType=Capsule, length=11.5269, radius=12.8328, shapeName=upperarm_r_capsule, center=<-18.4993 -3.551 0>, rotation=<rot 90 0 0>}]
}
constraints [
 {
  bone1Name=spine_05
  bone2Name=spine_04
  swing1LimitDegrees=45
  swing1Motion=Free
  swing2LimitDegrees=45
  swing2Motion=Free
  twistLimitDegrees=45
  twistMotion=Free
 }
 {
  bone1Name=clavicle_l
  bone2Name=spine_05
  swing1LimitDegrees=45
  swing1Motion=Free
  swing2LimitDegrees=45
  swing2Motion=Free
  twistLimitDegrees=45
  twistMotion=Free
 }
 {
  bone1Name=spine_04
  bone2Name=spine_03
  swing1LimitDegrees=45
  swing1Motion=Free
  swing2LimitDegrees=45
  swing2Motion=Free
  twistLimitDegrees=45
  twistMotion=Free
 }
 {
  bone1Name=upperarm_l
  bone2Name=clavicle_l
  swing1LimitDegrees=45
  swing1Motion=Free
  swing2LimitDegrees=45
  swing2Motion=Free
  twistLimitDegrees=45
  twistMotion=Free
 }
 {
  bone1Name=spine_02
  bone2Name=spine_01
  swing1LimitDegrees=45
  swing1Motion=Free
  swing2LimitDegrees=45
  swing2Motion=Free
  twistLimitDegrees=45
  twistMotion=Free
 }
 {
  bone1Name=lowerarm_l
  bone2Name=upperarm_l
  swing1LimitDegrees=45
  swing1Motion=Free
  swing2LimitDegrees=45
  swing2Motion=Free
  twistLimitDegrees=45
  twistMotion=Free
 }
 {
  bone1Name=clavicle_r
  bone2Name=spine_05
  swing1LimitDegrees=45
  swing1Motion=Free
  swing2LimitDegrees=45
  swing2Motion=Free
  twistLimitDegrees=45
  twistMotion=Free
 }
 {
  bone1Name=upperarm_r
  bone2Name=clavicle_r
  swing1LimitDegrees=45
  swing1Motion=Free
  swing2LimitDegrees=45
  swing2Motion=Free
  twistLimitDegrees=45
  twistMotion=Free
 }
 {
  bone1Name=lowerarm_r
  bone2Name=upperarm_r
  swing1LimitDegrees=45
  swing1Motion=Free
  swing2LimitDegrees=45
  swing2Motion=Free
  twistLimitDegrees=45
  twistMotion=Free
 }
 {
  bone1Name=thigh_l
  bone2Name=pelvis
  swing1LimitDegrees=45
  swing1Motion=Free
  swing2LimitDegrees=45
  swing2Motion=Free
  twistLimitDegrees=45
  twistMotion=Free
 }
 {
  bone1Name=calf_l
  bone2Name=thigh_l
  swing1LimitDegrees=45
  swing1Motion=Free
  swing2LimitDegrees=45
  swing2Motion=Free
  twistLimitDegrees=45
  twistMotion=Free
 }
 {
  bone1Name=thigh_r
  bone2Name=pelvis
  swing1LimitDegrees=45
  swing1Motion=Free
  swing2LimitDegrees=45
  swing2Motion=Free
  twistLimitDegrees=45
  twistMotion=Free
 }
 {
  bone1Name=calf_r
  bone2Name=thigh_r
  swing1LimitDegrees=45
  swing1Motion=Free
  swing2LimitDegrees=45
  swing2Motion=Free
  twistLimitDegrees=45
  twistMotion=Free
 }
 {
  bone1Name=spine_03
  bone2Name=spine_02
  swing1LimitDegrees=45
  swing1Motion=Free
  swing2LimitDegrees=45
  swing2Motion=Free
  twistLimitDegrees=45
  twistMotion=Free
 }
 {
  bone1Name=cape_l_MSTR
  bone2Name=clavicle_l
  swing1LimitDegrees=45
  swing1Motion=Free
  swing2LimitDegrees=45
  swing2Motion=Free
  twistLimitDegrees=45
  twistMotion=Free
 }
 {bone1Name=cape_01_l, bone2Name=cape_l_MSTR, swing1LimitDegrees=30, swing2LimitDegrees=45, twistLimitDegrees=0}
 {bone1Name=cape_02_l, bone2Name=cape_01_l, swing1LimitDegrees=30, swing2LimitDegrees=45, twistLimitDegrees=0}
 {bone1Name=cape_03_l, bone2Name=cape_02_l, swing1LimitDegrees=30, swing2LimitDegrees=45, twistLimitDegrees=0}
 {bone1Name=cape_04_l, bone2Name=cape_03_l, swing1LimitDegrees=30, swing2LimitDegrees=45, twistLimitDegrees=0}
 {bone1Name=cape_05_l, bone2Name=cape_04_l, swing1LimitDegrees=15, swing2LimitDegrees=20, twistLimitDegrees=0}
 {bone1Name=cape_06_l, bone2Name=cape_05_l, swing1LimitDegrees=20, swing2LimitDegrees=20, twistLimitDegrees=0}
 {
  bone1Name=cape_r_MSTR
  bone2Name=clavicle_r
  swing1LimitDegrees=45
  swing1Motion=Free
  swing2LimitDegrees=45
  swing2Motion=Free
  twistLimitDegrees=45
  twistMotion=Free
 }
 {
  bone1Name=cape_mid_l_MSTR
  bone2Name=spine_05
  swing1LimitDegrees=45
  swing1Motion=Free
  swing2LimitDegrees=45
  swing2Motion=Free
  twistLimitDegrees=45
  twistMotion=Free
 }
 {bone1Name=cape_mid_01_l, bone2Name=cape_mid_l_MSTR, swing1LimitDegrees=30, swing2LimitDegrees=45, twistLimitDegrees=0}
 {bone1Name=cape_mid_02_l, bone2Name=cape_mid_01_l, swing1LimitDegrees=30, swing2LimitDegrees=45, twistLimitDegrees=0}
 {bone1Name=cape_mid_03_l, bone2Name=cape_mid_02_l, swing1LimitDegrees=30, swing2LimitDegrees=45, twistLimitDegrees=0}
 {bone1Name=cape_mid_04_l, bone2Name=cape_mid_03_l, swing1LimitDegrees=30, swing2LimitDegrees=45, twistLimitDegrees=0}
 {bone1Name=cape_mid_05_l, bone2Name=cape_mid_04_l, swing1LimitDegrees=20, swing2LimitDegrees=20, twistLimitDegrees=0}
 {bone1Name=cape_mid_06_l, bone2Name=cape_mid_05_l, swing1LimitDegrees=20, swing2LimitDegrees=20, twistLimitDegrees=0}
 {
  bone1Name=cape_mid_r_MSTR
  bone2Name=spine_05
  swing1LimitDegrees=45
  swing1Motion=Free
  swing2LimitDegrees=45
  swing2Motion=Free
  twistLimitDegrees=45
  twistMotion=Free
 }
 {bone1Name=cape_mid_01_r, bone2Name=cape_mid_r_MSTR, swing1LimitDegrees=30, swing2LimitDegrees=45, twistLimitDegrees=0}
 {bone1Name=cape_mid_02_r, bone2Name=cape_mid_01_r, swing1LimitDegrees=30, swing2LimitDegrees=45, twistLimitDegrees=0}
 {bone1Name=cape_mid_03_r, bone2Name=cape_mid_02_r, swing1LimitDegrees=30, swing2LimitDegrees=45, twistLimitDegrees=0}
 {bone1Name=cape_mid_04_r, bone2Name=cape_mid_03_r, swing1LimitDegrees=30, swing2LimitDegrees=45, twistLimitDegrees=0}
 {bone1Name=cape_mid_05_r, bone2Name=cape_mid_04_r, swing1LimitDegrees=20, swing2LimitDegrees=20, twistLimitDegrees=0}
 {bone1Name=cape_mid_06_r, bone2Name=cape_mid_05_r, swing1LimitDegrees=20, swing2LimitDegrees=20, twistLimitDegrees=0}
 {
  bone1Name=foot_l
  bone2Name=calf_l
  swing1LimitDegrees=45
  swing1Motion=Free
  swing2LimitDegrees=45
  swing2Motion=Free
  twistLimitDegrees=45
  twistMotion=Free
 }
 {
  bone1Name=foot_r
  bone2Name=calf_r
  swing1LimitDegrees=45
  swing1Motion=Free
  swing2LimitDegrees=45
  swing2Motion=Free
  twistLimitDegrees=45
  twistMotion=Free
 }
 {bone1Name=cape_01_r, bone2Name=cape_r_MSTR, swing1LimitDegrees=30, swing2LimitDegrees=45, twistLimitDegrees=0}
 {bone1Name=cape_02_r, bone2Name=cape_01_r, swing1LimitDegrees=30, swing2LimitDegrees=45, twistLimitDegrees=0}
 {bone1Name=cape_03_r, bone2Name=cape_02_r, swing1LimitDegrees=30, swing2LimitDegrees=45, twistLimitDegrees=0}
 {bone1Name=cape_04_r, bone2Name=cape_03_r, swing1LimitDegrees=30, swing2LimitDegrees=45, twistLimitDegrees=0}
 {bone1Name=cape_05_r, bone2Name=cape_04_r, swing1LimitDegrees=45, swing2LimitDegrees=45, twistLimitDegrees=0}
 {bone1Name=cape_06_r, bone2Name=cape_05_r, swing1LimitDegrees=45, swing2LimitDegrees=45, twistLimitDegrees=0}
 {
  bone1Name=cape_mid_05_l
  bone2Name=cape_mid_05_r
  swing1LimitDegrees=45
  swing1Motion=Free
  swing2LimitDegrees=45
  swing2Motion=Free
  twistLimitDegrees=45
  twistMotion=Free
 }
 {
  bone1Name=cape_mid_05_r
  bone2Name=cape_mid_05_l
  swing1LimitDegrees=45
  swing1Motion=Free
  swing2LimitDegrees=45
  swing2Motion=Free
  twistLimitDegrees=45
  twistMotion=Free
 }
 {bone1Name=cape_mid_04_l, bone2Name=cape_04_l, swing1LimitDegrees=45, swing2LimitDegrees=45, twistLimitDegrees=45}
 {bone1Name=cape_mid_05_l, bone2Name=cape_05_l, swing1LimitDegrees=45, swing2LimitDegrees=45, twistLimitDegrees=45}
 {bone1Name=cape_mid_06_l, bone2Name=cape_06_l, swing1LimitDegrees=45, swing2LimitDegrees=45, twistLimitDegrees=45}
 {
  bone1Name=cape_mid_06_r
  bone2Name=cape_06_r
  swing1LimitDegrees=45
  swing1Motion=Free
  swing2LimitDegrees=45
  swing2Motion=Free
  twistLimitDegrees=45
  twistMotion=Free
 }
 {
  bone1Name=cape_mid_05_r
  bone2Name=cape_05_r
  swing1LimitDegrees=45
  swing1Motion=Free
  swing2LimitDegrees=45
  swing2Motion=Free
  twistLimitDegrees=45
  twistMotion=Free
 }
 {
  bone1Name=cape_mid_04_r
  bone2Name=cape_04_r
  swing1LimitDegrees=45
  swing1Motion=Free
  swing2LimitDegrees=45
  swing2Motion=Free
  twistLimitDegrees=45
  twistMotion=Free
 }
 {
  bone1Name=spine_01
  bone2Name=pelvis
  swing1LimitDegrees=45
  swing1Motion=Free
  swing2LimitDegrees=45
  swing2Motion=Free
  twistLimitDegrees=45
  twistMotion=Free
 }
]
