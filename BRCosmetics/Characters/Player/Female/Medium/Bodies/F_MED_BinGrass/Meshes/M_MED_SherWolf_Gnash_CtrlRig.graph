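# Control Rig for the SherWolf Gnash player body, adding procedural corrective motion to its bags, shoulder armor and cape.
# It runs a shared DeformRig function, then spherical pose readers measure thigh, arm, clavicle and spine poses to weight constraints and transform offsets on those bones.
# Spine controls aimed at foot controls feed additional readers for cape correctives.
# A final loop halves local Y and Z rotation on neck and head bones.
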ControlRigBlueprint /BRCosmetics/Characters/Player/Female/Medium/Bodies/F_MED_BinGrass/Meshes/M_MED_SherWolf_Gnash_CtrlRig
graphs [
 ContainedGraph {nodes [
   n1 = RigVMFunctionEntryNode()
   SequenceAggregate = &RigVMFunction_Sequence(ExecuteContext=@n1.ExecuteContext)
   SequenceAggregate_1 = &RigVMFunction_Sequence(ExecuteContext=@SequenceAggregate.B)
   SequenceAggregate_2 = &RigVMFunction_Sequence(ExecuteContext=@SequenceAggregate_1.B)
   SequenceAggregate_3 = &RigVMFunction_Sequence(ExecuteContext=@SequenceAggregate_2.B)
   RigVMFunctionReturnNode(A=@SequenceAggregate.A, B=@SequenceAggregate_1.A, C=@SequenceAggregate_2.A, D=@SequenceAggregate_3.A, E=@SequenceAggregate_3.B)
  ]}
 ContainedGraph {nodes [
   n1 = RigVMFunctionEntryNode()
   SequenceAggregate = &RigVMFunction_Sequence(ExecuteContext=@n1.ExecuteContext)
   SequenceAggregate_1 = &RigVMFunction_Sequence(ExecuteContext=@SequenceAggregate.B)
   SequenceAggregate_2 = &RigVMFunction_Sequence(ExecuteContext=@SequenceAggregate_1.B)
   RigVMFunctionReturnNode(A=@SequenceAggregate.A, B=@SequenceAggregate_1.A, C=@SequenceAggregate_2.A, D=@SequenceAggregate_2.B)
  ]}
 RigVMFunctionLibrary {links [], nodes []}
 RigVMModel {nodes [
   RigUnit_BeginExecution = &RigUnit_BeginExecution()
   FN_DeformRig_FNC = RigVMFunctionReferenceNode(
    DeformRig_ Curve="(deformRig_l_upperarm_up_alpha_72_#1=False,deformRig_r_upperarm_up_alpha_73_#2=False,deformRig_l_upperarm_fr_alpha_74_#3=False,deformRig_r_upperarm_fr_alpha_75_CB51B8 ...(truncated)"
    DeformRig_Bool="(upperarm_1_#4=True,elbow_3_#5=True,wrist_7_#6=True,groin_8_#7=True,glute_17_#8=True,knee_9_40C408324CA261 ...(truncated)"
    DeformRig_Debug=(upperarm_debug_18_#4=False,wrist_debug_20_#6=False,lat_debug_26_#9=False,groinGlute_debug_27_#7=False)
    DeformRig_Value="(upperarm_up_18_#10=(Rotation=(X=0.0,Y=-0.0,Z=0.0,W=1.0),Translation=(X=0,Y=0,Z=0),Scale3D=(X=0.70,Y=1.0,Z=1.0)),upperarm_fwd_fr_19_#11=(Rotation=(X=0.0,Y=-0.0,Z=0.0,W=1.0),Translation=(X=-5,Y=5,Z=- ...(truncated)"
    ExecuteContext=@RigUnit_BeginExecution.ExecutePin
   )
   > SequenceAggregate = RigVMAggregateNode()
   n1 = &RigUnit_SphericalPoseReader(
    ActiveRegionScaleFactors=(PositiveWidth=3.5,NegativeWidth=0.5,PositiveHeight=3,NegativeHeight=3)
    ActiveRegionSize=0.1
    Debug=(bDrawDebug=false,DebugScale=25,DebugSegments=20,DebugThickness=0.25)
    DriverAxis=(X=-1,Y=0,Z=0)
    DriverItem=(Type=Bone,Name="thigh_twist_01_l")
    FalloffRegionScaleFactors=(PositiveWidth=1,NegativeWidth=0.8,PositiveHeight=1,NegativeHeight=1)
    FalloffSize=0.55
    OptionalParentItem=(Type=Bone,Name="spine_01")
    RotationOffset=(X=90,Y=0,Z=-15)
    ExecutePin=@SequenceAggregate.A
   )
   > RigVMVariableNode(Value=@n1.OutputParam) {variable=spr_thigh_l_up_alpha}
   > n2 = &RigUnit_SphericalPoseReader(
    ActiveRegionScaleFactors=(PositiveWidth=3.5,NegativeWidth=3,PositiveHeight=3,NegativeHeight=3)
    ActiveRegionSize=0.1
    Debug=(bDrawDebug=false,DebugScale=25,DebugSegments=20,DebugThickness=0.25)
    DriverAxis=(X=1,Y=0,Z=0)
    DriverItem=(Type=Bone,Name="thigh_twist_01_r")
    FalloffRegionScaleFactors=(PositiveWidth=1,NegativeWidth=1,PositiveHeight=1,NegativeHeight=1)
    FalloffSize=0.55
    FlipHeightScaling=false
    FlipWidthScaling=true
    OptionalParentItem=(Type=Bone,Name="spine_01")
    RotationOffset=(X=-90,Y=0,Z=-15)
   )
   > RigVMVariableNode(Value=@n2.OutputParam) {variable=spr_thigh_r_up_alpha}
   > n3 = &RigUnit_SphericalPoseReader(
    ActiveRegionScaleFactors=(PositiveWidth=2,NegativeWidth=1,PositiveHeight=1,NegativeHeight=1)
    ActiveRegionSize=0.1
    Debug=(bDrawDebug=false,DebugScale=25,DebugSegments=20,DebugThickness=0.25)
    DriverAxis=(X=1,Y=0,Z=0)
    DriverItem=(Type=Bone,Name="upperarm_twist_01_l")
    FalloffRegionScaleFactors=(PositiveWidth=1,NegativeWidth=1,PositiveHeight=0.9,NegativeHeight=0.9)
    FalloffSize=0.7
    OptionalParentItem=(Type=Bone)
    RotationOffset=(X=0,Y=10,Z=0)
   )
   > RigVMVariableNode(Value=@n3.OutputParam) {variable=spr_upperarm_l_up_alpha}
   > n4 = &RigUnit_SphericalPoseReader(
    ActiveRegionScaleFactors=(PositiveWidth=2,NegativeWidth=1,PositiveHeight=1,NegativeHeight=1)
    ActiveRegionSize=0.1
    Debug=(bDrawDebug=false,DebugScale=25,DebugSegments=20,DebugThickness=0.25)
    DriverAxis=(X=-1,Y=0,Z=0)
    DriverItem=(Type=Bone,Name="upperarm_twist_01_r")
    FalloffRegionScaleFactors=(PositiveWidth=1,NegativeWidth=1,PositiveHeight=0.9,NegativeHeight=0.9)
    FalloffSize=0.7
    FlipHeightScaling=true
    FlipWidthScaling=false
    OptionalParentItem=(Type=Bone)
    RotationOffset=(X=180,Y=-10,Z=180)
   )
   > RigVMVariableNode(Value=@n4.OutputParam) {variable=spr_upperarm_r_up_alpha}
   &RigUnit_RotationConstraintLocalSpaceOffset(
    AdvancedSettings=(InterpolationType=Shortest,RotationOrderForFilter=XZY)
    Child=(Type=Bone,Name="dyn_bag01")
    Filter=(bX=True,bY=True,bZ=True)
    Parents=((Item=(Type=Bone,Name="pelvis"),Weight=0.5),(Item=(Type=Bone,Name="thigh_twist_01_l"),Weight=0.5))
    Weight=1
    ExecutePin=@SequenceAggregate.B
   )
   > &RigUnit_RotationConstraintLocalSpaceOffset(
    AdvancedSettings=(InterpolationType=Shortest,RotationOrderForFilter=XZY)
    Child=(Type=Bone,Name="dyn_bag02")
    Filter=(bX=True,bY=True,bZ=True)
    Parents=((Item=(Type=Bone,Name="pelvis"),Weight=0.5),(Item=(Type=Bone,Name="thigh_twist_01_l"),Weight=0.5))
    Weight=1
   )
   > &RigUnit_ModifyTransforms(
    ItemToModify="((Item=(Type=Bone,Name=\"dyn_bag01\"),Transform=(Rotation=(X=0.0545,Y=-0.1853,Z=-0.1555,W=0.9688),Translation=(X=-2,Y=0,Z=1),Scale3D=(X=1,Y=1,Z=1))),(Item=(Type=Bone,Name=\"dyn_bag02\"),Transform=(Rotation=(X=0,Y=-0.0436,Z=0,W=0.999),Translation=(X=0,Y=-1,Z=1) ...(truncated)"
    Weight=RigVMVariableNode() {variable=spr_thigh_l_up_alpha}.Value
   )
   > &RigUnit_ParentConstraint(
    AdvancedSettings=(InterpolationType=Average,RotationOrderForFilter=XZY)
    Child=(Type=Bone,Name="dyn_bag03")
    Filter=(TranslationFilter=(bX=false,bY=false,bZ=false),RotationFilter=(bX=True,bY=True,bZ=True),ScaleFilter=(bX=True,bY=True,bZ=True))
    Parents=((Item=(Type=Bone,Name="pelvis"),Weight=0.5),(Item=(Type=Bone,Name="thigh_twist_01_r"),Weight=0.5))
    Weight=1
   )
   > &RigUnit_ParentConstraint(
    AdvancedSettings=(InterpolationType=Average,RotationOrderForFilter=XZY)
    Child=(Type=Bone,Name="dyn_bag04")
    Filter=(TranslationFilter=(bX=false,bY=false,bZ=false),RotationFilter=(bX=True,bY=True,bZ=True),ScaleFilter=(bX=True,bY=True,bZ=True))
    Parents=((Item=(Type=Bone,Name="pelvis"),Weight=0.5),(Item=(Type=Bone,Name="thigh_twist_01_r"),Weight=0.5))
    Weight=1
   )
   > &RigUnit_ModifyTransforms(
    ItemToModify="((Item=(Type=Bone,Name=\"dyn_bag03\"),Transform=(Rotation=(X=0,Y=-0.3007,Z=0,W=0.9537),Translation=(X=-1,Y=0,Z=0),Scale3D=(X=1,Y=1,Z=1))),(Item=(Type=Bone,Name=\"dyn_bag04\"),Transform=(Rotation=(X=0,Y=-0.1305,Z=0,W=0.9914),Translation=(X=0,Y=0,Z=0),Scale3D=(X ...(truncated)"
    Weight=RigVMVariableNode() {variable=spr_thigh_r_up_alpha}.Value
   )
   > &RigUnit_ParentConstraint(
    AdvancedSettings=(RotationOrderForFilter=XZY)
    Child=(Type=Bone,Name="dyn_shoulder_armor_l_MSTR")
    Filter=(TranslationFilter=(bX=true,bY=true,bZ=true),RotationFilter=(bX=True,bY=True,bZ=True),ScaleFilter=(bX=True,bY=True,bZ=True))
    Parents=((Item=(Type=Bone,Name="upperarm_twist_01_l"),Weight=1),(Item=(Type=Bone,Name="clavicle_l"),Weight=0))
    Weight=1
   )
   > &RigUnit_ModifyTransforms(
    ItemToModify="((Item=(Type=Bone,Name=\"dyn_shoulder_armor_l_Aim\"),Transform=(Rotation=(X=-0.0436,Y=-0.0436,Z=-0.0019,W=0.9981),Translation=(X=-0.8,Y=-1,Z=0),Scale3D=(X=1,Y=1,Z=1))),(Item=(Type=Bone,Name=\"cape_l_MSTR\"),Transform=(Rotation=(X=0,Y=0,Z=0,W=1),Translation=(X= ...(truncated)"
    Weight=1
   )
   > &RigUnit_ModifyTransforms(
    ItemToModify="((Item=(Type=Bone,Name=\"dyn_shoulder_armor_l_Aim\"),Transform=(Rotation=(X=0,Y=0,Z=0,W=1),Translation=(X=0,Y=-12,Z=-7),Scale3D=(X=1,Y=1,Z=1))),(Item=(Type=Bone,Name=\"cape_l_MSTR\"),Transform=(Rotation=(X=0,Y=0,Z=0,W=1),Translation=(X=0,Y=0,Z=0),Scale3D=(X=1, ...(truncated)"
    ItemToModify.0.Transform.Rotation=&RigVMFunction_MathQuaternionFromEuler(Euler=(X=120,Y=5,Z=15)).Result
    Weight=RigVMVariableNode() {variable=spr_upperarm_l_up_alpha}.Value
   )
   > &RigUnit_ParentConstraint(
    AdvancedSettings=(RotationOrderForFilter=XZY)
    Child=(Type=Bone,Name="dyn_shoulder_armor_r_MSTR")
    Filter=(TranslationFilter=(bX=true,bY=true,bZ=true),RotationFilter=(bX=True,bY=True,bZ=True),ScaleFilter=(bX=True,bY=True,bZ=True))
    Parents=((Item=(Type=Bone,Name="upperarm_twist_01_r"),Weight=1),(Item=(Type=Bone,Name="clavicle_r"),Weight=0))
    Weight=1
   )
   > &RigUnit_ModifyTransforms(
    ItemToModify=((Item=(Type=Bone,Name="dyn_shoulder_armor_r_Aim"),Transform=(Rotation=(X=0,Y=-0.0436,Z=0,W=0.999),Translation=(X=0.8,Y=1,Z=0),Scale3D=(X=1,Y=1,Z=1))))
    Weight=1
   )
   > &RigUnit_ModifyTransforms(
    ItemToModify=((Item=(Type=Bone,Name="dyn_shoulder_armor_r_Aim"),Transform=(Rotation=(X=0,Y=0,Z=0,W=1),Translation=(X=0,Y=12,Z=7),Scale3D=(X=1,Y=1,Z=1))))
    ItemToModify.0.Transform.Rotation=&RigVMFunction_MathQuaternionFromEuler(Euler=(X=120,Y=5,Z=15)).Result
    Weight=RigVMVariableNode() {variable=spr_upperarm_r_up_alpha}.Value
   )
   &RigUnit_ParentConstraint(
    AdvancedSettings=(RotationOrderForFilter=XZY)
    Child=(Type=Bone,Name="cape_mid_l_MSTR")
    Filter=(TranslationFilter=(bX=true,bY=true,bZ=true),RotationFilter=(bX=True,bY=True,bZ=True),ScaleFilter=(bX=True,bY=True,bZ=True))
    Parents=((Item=(Type=Bone,Name="spine_04"),Weight=0.6),(Item=(Type=Bone,Name="spine_05"),Weight=0.4))
    Weight=1
    ExecutePin=@SequenceAggregate.C
   )
   > &RigUnit_ParentConstraint(
    AdvancedSettings=(RotationOrderForFilter=XZY)
    Child=(Type=Bone,Name="cape_l_MSTR")
    Filter=(TranslationFilter=(bX=true,bY=true,bZ=true),RotationFilter=(bX=True,bY=True,bZ=True),ScaleFilter=(bX=True,bY=True,bZ=True))
    Parents=((Item=(Type=Bone,Name="spine_04"),Weight=0.1),(Item=(Type=Bone,Name="spine_05"),Weight=0.9))
    Weight=1
   )
   > n5 = &RigUnit_SphericalPoseReader(
    ActiveRegionScaleFactors=(PositiveWidth=0.3,NegativeWidth=2,PositiveHeight=1,NegativeHeight=0.8)
    ActiveRegionSize=0.1
    Debug=(bDrawDebug=false,DebugScale=25,DebugSegments=20,DebugThickness=0.25)
    DriverAxis=(X=1,Y=0,Z=0)
    DriverItem=(Type=Bone,Name="upperarm_l")
    FalloffRegionScaleFactors=(PositiveWidth=0.8,NegativeWidth=2,PositiveHeight=1,NegativeHeight=1)
    FalloffSize=0.4
    FlipWidthScaling=false
    OptionalParentItem=(Type=Bone,Name="spine_05")
    RotationOffset=(X=180,Y=-90,Z=100)
   )
   > &RigUnit_ModifyTransforms(
    ItemToModify="((Item=(Type=Bone,Name=\"cape_l_MSTR\"),Transform=(Rotation=(X=0,Y=0,Z=0,W=1),Translation=(X=0,Y=5,Z=0),Scale3D=(X=1,Y=1,Z=1))),(Item=(Type=Bone,Name=\"cape_mid_l_MSTR\"),Transform=(Rotation=(X=0,Y=-0.0872,Z=0,W=0.9962),Translation=(X=-5,Y=-5,Z=5),Scale3D=(X=1 ...(truncated)"
    Weight=@n5.OutputParam
   )
   > n6 = &RigUnit_SphericalPoseReader(
    ActiveRegionScaleFactors=(PositiveWidth=0.5,NegativeWidth=0.2,PositiveHeight=0.1,NegativeHeight=0.5)
    ActiveRegionSize=0.1
    Debug=(bDrawDebug=false,DebugScale=25,DebugSegments=20,DebugThickness=0.25)
    DriverAxis=(X=1,Y=0,Z=0)
    DriverItem=(Type=Bone,Name="clavicle_l")
    FalloffRegionScaleFactors=(PositiveWidth=1,NegativeWidth=0.5,PositiveHeight=0.15,NegativeHeight=0.6)
    FalloffSize=0.6
    FlipWidthScaling=false
    OptionalParentItem=(Type=Bone,Name="pelvis")
    RotationOffset=(X=0,Y=-70,Z=90)
   )
   > &RigUnit_ModifyTransforms(
    ItemToModify=((Item=(Type=Bone,Name="cape_l_MSTR"),Transform=(Rotation=(X=-0.2156,Y=0.0189,Z=-0.0851,W=0.9726),Translation=(X=0,Y=0,Z=0),Scale3D=(X=1,Y=1,Z=1))))
    Weight=@n6.OutputParam
   )
   > &RigUnit_ParentConstraint(
    AdvancedSettings=(RotationOrderForFilter=XZY)
    Child=(Type=Bone,Name="cape_mid_r_MSTR")
    Filter=(TranslationFilter=(bX=true,bY=true,bZ=true),RotationFilter=(bX=True,bY=True,bZ=True),ScaleFilter=(bX=True,bY=True,bZ=True))
    Parents=((Item=(Type=Bone,Name="spine_04"),Weight=0.6),(Item=(Type=Bone,Name="spine_05"),Weight=0.4))
    Weight=1
   )
   > &RigUnit_ParentConstraint(
    AdvancedSettings=(RotationOrderForFilter=XZY)
    Child=(Type=Bone,Name="cape_r_MSTR")
    Filter=(TranslationFilter=(bX=true,bY=true,bZ=true),RotationFilter=(bX=True,bY=True,bZ=True),ScaleFilter=(bX=True,bY=True,bZ=True))
    Parents=((Item=(Type=Bone,Name="spine_04"),Weight=0.1),(Item=(Type=Bone,Name="spine_05"),Weight=0.9))
    Weight=1
   )
   > n7 = &RigUnit_SphericalPoseReader(
    ActiveRegionScaleFactors=(PositiveWidth=0.3,NegativeWidth=2,PositiveHeight=1,NegativeHeight=0.8)
    ActiveRegionSize=0.1
    Debug=(bDrawDebug=false,bDraw2D=false,DebugScale=25,DebugSegments=20,DebugThickness=0.25)
    DriverAxis=(X=-1,Y=0,Z=0)
    DriverItem=(Type=Bone,Name="upperarm_r")
    FalloffRegionScaleFactors=(PositiveWidth=0.8,NegativeWidth=2,PositiveHeight=1,NegativeHeight=1)
    FalloffSize=0.4
    FlipHeightScaling=false
    FlipWidthScaling=true
    OptionalParentItem=(Type=Bone,Name="spine_05")
    RotationOffset=(X=0,Y=-90,Z=100)
   )
   > &RigUnit_ModifyTransforms(
    ItemToModify="((Item=(Type=Bone,Name=\"cape_r_MSTR\"),Transform=(Rotation=(X=-0.0734,Y=-0.4162,Z=-0.1574,W=0.8925),Translation=(X=0,Y=-5,Z=0),Scale3D=(X=1,Y=1,Z=1))),(Item=(Type=Bone,Name=\"cape_mid_r_MSTR\"),Transform=(Rotation=(X=0,Y=-0.2588,Z=0,W=0.9659),Translation=(X=5 ...(truncated)"
    Weight=@n7.OutputParam
   )
   > n8 = &RigUnit_SphericalPoseReader(
    ActiveRegionScaleFactors=(PositiveWidth=0.5,NegativeWidth=0.2,PositiveHeight=0.1,NegativeHeight=0.5)
    ActiveRegionSize=0.1
    Debug=(bDrawDebug=false,bDraw2D=false,DebugScale=25,DebugSegments=20,DebugThickness=0.25)
    DriverAxis=(X=-1,Y=0,Z=0)
    DriverItem=(Type=Bone,Name="clavicle_r")
    FalloffRegionScaleFactors=(PositiveWidth=1,NegativeWidth=0.5,PositiveHeight=0.15,NegativeHeight=0.6)
    FalloffSize=0.6
    FlipHeightScaling=true
    FlipWidthScaling=true
    OptionalParentItem=(Type=Bone,Name="pelvis")
    RotationOffset=(X=180,Y=-70,Z=90)
   )
   > &RigUnit_ModifyTransforms(
    ItemToModify=((Item=(Type=Bone,Name="cape_r_MSTR"),Transform=(Rotation=(X=-0.2156,Y=0.0189,Z=-0.0851,W=0.9726),Translation=(X=0,Y=0,Z=0),Scale3D=(X=1,Y=1,Z=1))))
    Weight=@n8.OutputParam
   )
   > n9 = &RigUnit_SphericalPoseReader(
    ActiveRegionScaleFactors=(PositiveWidth=1,NegativeWidth=1,PositiveHeight=1,NegativeHeight=0.1)
    ActiveRegionSize=0.1
    Debug=(bDrawDebug=false,DebugScale=25,DebugSegments=20,DebugThickness=0.25)
    DriverAxis=(X=1,Y=0,Z=0)
    DriverItem=(Type=Bone,Name="upperarm_l")
    FalloffRegionScaleFactors=(PositiveWidth=1,NegativeWidth=1.5,PositiveHeight=1.8,NegativeHeight=1)
    FalloffSize=0.5
    FlipWidthScaling=true
    OptionalParentItem=(Type=Bone,Name="None")
    RotationOffset=(X=-20,Y=10,Z=0)
   )
   > &RigUnit_ParentConstraint(
    AdvancedSettings=(RotationOrderForFilter=XZY)
    Child=(Type=Bone,Name="clavicle_cape_l")
    Filter=(TranslationFilter=(bX=true,bY=true,bZ=true),RotationFilter=(bX=True,bY=True,bZ=True),ScaleFilter=(bX=True,bY=True,bZ=True))
    Parents=((Item=(Type=Bone,Name="upperarm_l"),Weight=0.5),(Item=(Type=Bone,Name="clavicle_l"),Weight=0.5))
    Weight=@n9.OutputParam
   )
   > &RigUnit_ModifyTransforms(
    ItemToModify="((Item=(Type=Bone,Name=\"clavicle_cape_l\"),Transform=(Rotation=(X=-0.31,Y=-0.1445,Z=0.3971,W=0.8517),Translation=(X=5,Y=-7,Z=9),Scale3D=(X=1,Y=1,Z=1))),(Item=(Type=Bone,Name=\"cape_l_MSTR\"),Transform=(Rotation=(X=0,Y=0,Z=0,W=1),Translation=(X=0,Y=-10,Z=0),Sc ...(truncated)"
    Weight=@n9.OutputParam
   )
   > n10 = &RigUnit_SphericalPoseReader(
    ActiveRegionScaleFactors=(PositiveWidth=1,NegativeWidth=1,PositiveHeight=1,NegativeHeight=0.1)
    ActiveRegionSize=0.1
    Debug=(bDrawDebug=false,DebugScale=25,DebugSegments=20,DebugThickness=0.25)
    DriverAxis=(X=-1,Y=0,Z=0)
    DriverItem=(Type=Bone,Name="upperarm_r")
    FalloffRegionScaleFactors=(PositiveWidth=1,NegativeWidth=1.5,PositiveHeight=1.8,NegativeHeight=1)
    FalloffSize=0.55
    FlipHeightScaling=false
    FlipWidthScaling=false
    OptionalParentItem=(Type=Bone)
    RotationOffset=(X=160,Y=10,Z=0)
   )
   > &RigUnit_ParentConstraint(
    AdvancedSettings=(RotationOrderForFilter=XZY)
    Child=(Type=Bone,Name="clavicle_cape_r")
    Filter=(TranslationFilter=(bX=true,bY=true,bZ=true),RotationFilter=(bX=True,bY=True,bZ=True),ScaleFilter=(bX=True,bY=True,bZ=True))
    Parents=((Item=(Type=Bone,Name="upperarm_r"),Weight=0.5),(Item=(Type=Bone,Name="clavicle_r"),Weight=0.5))
    Weight=@n10.OutputParam
   )
   > &RigUnit_ModifyTransforms(
    ItemToModify="((Item=(Type=Bone,Name=\"clavicle_cape_r\"),Transform=(Rotation=(X=0.31,Y=-0.1445,Z=-0.3971,W=0.8517),Translation=(X=5,Y=7,Z=5),Scale3D=(X=1,Y=1,Z=1))),(Item=(Type=Bone,Name=\"cape_r_MSTR\"),Transform=(Rotation=(X=0,Y=0,Z=0,W=1),Translation=(X=0,Y=10,Z=0),Scal ...(truncated)"
    Weight=@n10.OutputParam
   )
   > &RigUnit_ParentConstraint(
    AdvancedSettings=(RotationOrderForFilter=XZY)
    Child=(Type=Bone,Name="cape_01_l")
    Filter=(TranslationFilter=(bX=true,bY=true,bZ=true),RotationFilter=(bX=True,bY=True,bZ=True),ScaleFilter=(bX=True,bY=True,bZ=True))
    Parents=((Item=(Type=Bone,Name="cape_01_l"),Weight=0.5),(Item=(Type=Bone,Name="cape_mid_01_l"),Weight=0.5))
    Weight=1
   )
   > &RigUnit_ParentConstraint(
    AdvancedSettings=(RotationOrderForFilter=XZY)
    Child=(Type=Bone,Name="cape_01_r")
    Filter=(TranslationFilter=(bX=true,bY=true,bZ=true),RotationFilter=(bX=True,bY=True,bZ=True),ScaleFilter=(bX=True,bY=True,bZ=True))
    Parents=((Item=(Type=Bone,Name="cape_01_r"),Weight=0.5),(Item=(Type=Bone,Name="cape_mid_01_r"),Weight=0.5))
    Weight=1
   )
   > n11 = &RigUnit_SphericalPoseReader(
    ActiveRegionScaleFactors=(PositiveWidth=0.1,NegativeWidth=3,PositiveHeight=1.5,NegativeHeight=2)
    ActiveRegionSize=0.1
    Debug=(bDrawDebug=false,DebugScale=25,DebugSegments=20,DebugThickness=0.25)
    DriverAxis=(X=1,Y=0,Z=0)
    DriverItem=(Type=Bone,Name="upperarm_l")
    FalloffRegionScaleFactors=(PositiveWidth=1,NegativeWidth=2,PositiveHeight=2,NegativeHeight=2)
    FalloffSize=0.45
    FlipWidthScaling=false
    OptionalParentItem=(Type=Bone,Name="None")
    RotationOffset=(X=-30,Y=-50,Z=-70)
   )
   > &RigUnit_ModifyTransforms(
    ItemToModify=((Item=(Type=Bone,Name="clavicle_cape_l"),Transform=(Rotation=(X=0.3415,Y=0.2273,Z=-0.3936,W=0.8227),Translation=(X=5,Y=0,Z=5),Scale3D=(X=1,Y=1,Z=1))))
    Weight=@n11.OutputParam
   )
   > n12 = &RigUnit_SphericalPoseReader(
    ActiveRegionScaleFactors=(PositiveWidth=0.1,NegativeWidth=3,PositiveHeight=1.5,NegativeHeight=2)
    ActiveRegionSize=0.1
    Debug=(bDrawDebug=false,bDraw2D=false,DebugScale=25,DebugSegments=20,DebugThickness=0.25)
    DriverAxis=(X=-1,Y=0,Z=0)
    DriverItem=(Type=Bone,Name="upperarm_r")
    FalloffRegionScaleFactors=(PositiveWidth=1,NegativeWidth=2,PositiveHeight=2,NegativeHeight=2)
    FalloffSize=0.45
    FlipHeightScaling=false
    FlipWidthScaling=true
    OptionalParentItem=(Type=Bone,Name="None")
    RotationOffset=(X=150,Y=-50,Z=-70)
   )
   > &RigUnit_ModifyTransforms(
    ItemToModify=((Item=(Type=Bone,Name="clavicle_cape_r"),Transform=(Rotation=(X=-0.3415,Y=0.2273,Z=0.3936,W=0.8227),Translation=(X=5,Y=0,Z=5),Scale3D=(X=1,Y=1,Z=1))))
    Weight=@n12.OutputParam
   )
   > n13 = &RigUnit_SphericalPoseReader(
    ActiveRegionScaleFactors=(PositiveWidth=1,NegativeWidth=2,PositiveHeight=1,NegativeHeight=1)
    ActiveRegionSize=0.1
    Debug=(bDrawDebug=false,DebugScale=25,DebugSegments=20,DebugThickness=0.25)
    DriverAxis=(X=1,Y=0,Z=0)
    DriverItem=(Type=Bone,Name="upperarm_l")
    FalloffRegionScaleFactors=(PositiveWidth=0.1,NegativeWidth=1.5,PositiveHeight=1,NegativeHeight=1.2)
    FalloffSize=0.45
    FlipWidthScaling=true
    OptionalParentItem=(Type=Bone,Name="None")
    RotationOffset=(X=-60,Y=-20,Z=0)
   )
   > &RigUnit_ModifyTransforms(
    ItemToModify="((Item=(Type=Bone,Name=\"cape_l_MSTR\"),Transform=(Rotation=(X=0,Y=0,Z=0,W=1),Translation=(X=20,Y=0,Z=0),Scale3D=(X=1,Y=1,Z=1))),(Item=(Type=Bone,Name=\"cape_mid_l_MSTR\"),Transform=(Rotation=(X=0,Y=0,Z=0,W=1),Translation=(X=0,Y=0,Z=10),Scale3D=(X=1,Y=1,Z=1))) ...(truncated)"
    Weight=@n13.OutputParam
   )
   > n14 = &RigUnit_SphericalPoseReader(
    ActiveRegionScaleFactors=(PositiveWidth=1,NegativeWidth=2,PositiveHeight=1,NegativeHeight=1)
    ActiveRegionSize=0.1
    Debug=(bDrawDebug=false,DebugScale=25,DebugSegments=20,DebugThickness=0.25)
    DriverAxis=(X=-1,Y=0,Z=0)
    DriverItem=(Type=Bone,Name="upperarm_r")
    FalloffRegionScaleFactors=(PositiveWidth=0.1,NegativeWidth=1.5,PositiveHeight=1,NegativeHeight=1.2)
    FalloffSize=0.45
    FlipHeightScaling=false
    FlipWidthScaling=false
    OptionalParentItem=(Type=Bone)
    RotationOffset=(X=110,Y=-20,Z=0)
   )
   > &RigUnit_ModifyTransforms(
    ItemToModify="((Item=(Type=Bone,Name=\"cape_r_MSTR\"),Transform=(Rotation=(X=0,Y=0,Z=0,W=1),Translation=(X=20,Y=0,Z=0),Scale3D=(X=1,Y=1,Z=1))),(Item=(Type=Bone,Name=\"cape_mid_r_MSTR\"),Transform=(Rotation=(X=0,Y=0,Z=0,W=1),Translation=(X=0,Y=0,Z=10),Scale3D=(X=1,Y=1,Z=1))) ...(truncated)"
    Weight=@n14.OutputParam
   )
   > SequenceAggregate_1 = RigVMAggregateNode()
   n15 = &RigUnit_SphericalPoseReader(
    ActiveRegionScaleFactors=(PositiveWidth=1,NegativeWidth=1,PositiveHeight=3,NegativeHeight=3)
    ActiveRegionSize=0.1
    Debug=(bDrawDebug=false,DebugScale=25,DebugSegments=20,DebugThickness=0.25)
    DriverAxis=(X=1,Y=0,Z=0)
    DriverItem=(Type=Bone,Name="spine_04")
    FalloffRegionScaleFactors=(PositiveWidth=0.5,NegativeWidth=0.5,PositiveHeight=0.7,NegativeHeight=0.7)
    FalloffSize=0.57
    FlipHeightScaling=false
    FlipWidthScaling=false
    OptionalParentItem=(Type=Bone,Name="spine_02")
    RotationOffset=(X=5,Y=90,Z=90)
    ExecutePin=@SequenceAggregate_1.A
   )
   > &RigUnit_ModifyTransforms(
    ItemToModify=((Item=(Type=Bone,Name="cape_mid_02_l"),Transform=(Rotation=(X=0,Y=-0.4226,Z=0,W=0.9063),Translation=(X=0,Y=0,Z=0),Scale3D=(X=1,Y=1,Z=1))))
    Weight=@n15.OutputParam
   )
   > &RigUnit_AimConstraintLocalSpaceOffset(
    AdvancedSettings=(DebugSettings=(bEnabled=false,Scale=10,WorldOffset=(Rotation=(X=0,Y=0,Z=0,W=1),Translation=(X=0,Y=0,Z=0),Scale3D=(X=1,Y=1,Z=1))),RotationOrderForFilter=XZY)
    Child=(Type=Control,Name="spine_05_ctrl")
    Filter=(bX=True,bY=True,bZ=True)
    Parents=((Item=(Type=Control,Name="foot_l_ctrl"),Weight=1))
    Weight=1
    WorldUp=(Target=(X=0,Y=0,Z=1),Kind=Direction,Space=(Type=None,Name="None"))
    bMaintainOffset=true
   )
   > n16 = &RigUnit_SphericalPoseReader(
    ActiveRegionScaleFactors=(PositiveWidth=0.1,NegativeWidth=0.1,PositiveHeight=0.1,NegativeHeight=1)
    ActiveRegionSize=0.1
    Debug=(bDrawDebug=false,DebugScale=25,DebugSegments=20,DebugThickness=0.25)
    DriverAxis=(X=-1,Y=0,Z=0)
    DriverItem=(Type=Control,Name="spine_05_ctrl_2")
    FalloffRegionScaleFactors=(PositiveWidth=1,NegativeWidth=1,PositiveHeight=1,NegativeHeight=2)
    FalloffSize=0.33
    FlipHeightScaling=true
    FlipWidthScaling=false
    OptionalParentItem=(Type=Bone,Name="spine_05")
    RotationOffset=(X=180,Y=-30,Z=-30)
   )
   > &RigUnit_ParentConstraint(
    AdvancedSettings=(RotationOrderForFilter=XZY)
    Child=(Type=Bone,Name="cape_01_l")
    Filter=(TranslationFilter=(bX=false,bY=false,bZ=false),RotationFilter=(bX=true,bY=true,bZ=true),ScaleFilter=(bX=True,bY=True,bZ=True))
    Parents=((Item=(Type=Bone,Name="thigh_twist_01_l"),Weight=0),(Item=(Type=Bone,Name="pelvis"),Weight=1))
    Weight=@n16.OutputParam
   )
   > &RigUnit_ModifyTransforms(
    ItemToModify="((Item=(Type=Bone,Name=\"cape_01_l\"),Transform=(Rotation=(X=0.0227,Y=-0.1285,Z=0.1722,W=0.9764),Translation=(X=0,Y=0,Z=0),Scale3D=(X=1,Y=1,Z=1))),(Item=(Type=Bone,Name=\"cape_mid_01_l\"),Transform=(Rotation=(X=0.0227,Y=-0.1285,Z=0.1722,W=0.9764),Translation=( ...(truncated)"
    Weight=@n16.OutputParam
   )
   > n17 = &RigUnit_SphericalPoseReader(
    ActiveRegionScaleFactors=(PositiveWidth=1,NegativeWidth=1,PositiveHeight=1.5,NegativeHeight=1)
    ActiveRegionSize=0.1
    Debug=(bDrawDebug=false,DebugScale=25,DebugSegments=20,DebugThickness=0.25)
    DriverAxis=(X=-1,Y=0,Z=0)
    DriverItem=(Type=Control,Name="spine_05_ctrl")
    FalloffRegionScaleFactors=(PositiveWidth=1,NegativeWidth=1,PositiveHeight=1,NegativeHeight=1)
    FalloffSize=0.3
    FlipHeightScaling=true
    FlipWidthScaling=false
    OptionalParentItem=(Type=Bone,Name="spine_05")
    RotationOffset=(X=0,Y=-100,Z=-120)
   )
   > &RigUnit_ModifyTransforms(
    ItemToModify="((Item=(Type=Bone,Name=\"cape_02_l\"),Transform=(Rotation=(X=0.067,Y=-0.25,Z=0.25,W=0.933),Translation=(X=0,Y=0,Z=0),Scale3D=(X=1,Y=1,Z=1))),(Item=(Type=Bone,Name=\"cape_mid_02_l\"),Transform=(Rotation=(X=-0.0151,Y=-0.173,Z=-0.0858,W=0.9811),Translation=(X=0,Y ...(truncated)"
    Weight=@n17.OutputParam
   )
   n18 = &RigUnit_SphericalPoseReader(
    ActiveRegionScaleFactors=(PositiveWidth=1,NegativeWidth=1,PositiveHeight=3,NegativeHeight=3)
    ActiveRegionSize=0.1
    Debug=(bDrawDebug=false,DebugScale=25,DebugSegments=20,DebugThickness=0.25)
    DriverAxis=(X=1,Y=0,Z=0)
    DriverItem=(Type=Bone,Name="spine_04")
    FalloffRegionScaleFactors=(PositiveWidth=0.5,NegativeWidth=0.5,PositiveHeight=0.7,NegativeHeight=0.7)
    FalloffSize=0.57
    FlipHeightScaling=false
    FlipWidthScaling=true
    OptionalParentItem=(Type=Bone,Name="spine_02")
    RotationOffset=(X=5,Y=90,Z=90)
    ExecutePin=@SequenceAggregate_1.B
   )
   > &RigUnit_ModifyTransforms(
    ItemToModify=((Item=(Type=Bone,Name="cape_mid_02_r"),Transform=(Rotation=(X=0,Y=-0.4226,Z=0,W=0.9063),Translation=(X=0,Y=0,Z=0),Scale3D=(X=1,Y=1,Z=1))))
    Weight=@n18.OutputParam
   )
   > &RigUnit_AimConstraintLocalSpaceOffset(
    AdvancedSettings=(DebugSettings=(bEnabled=false,Scale=10,WorldOffset=(Rotation=(X=0,Y=0,Z=0,W=1),Translation=(X=0,Y=0,Z=0),Scale3D=(X=1,Y=1,Z=1))),RotationOrderForFilter=XZY)
    Child=(Type=Control,Name="spine_05_ctrl_2")
    Filter=(bX=True,bY=True,bZ=True)
    Parents=((Item=(Type=Control,Name="foot_r_ctrl"),Weight=1))
    Weight=1
    WorldUp=(Target=(X=0,Y=0,Z=1),Kind=Direction,Space=(Type=None,Name="None"))
    bMaintainOffset=true
   )
   > n19 = &RigUnit_SphericalPoseReader(
    ActiveRegionScaleFactors=(PositiveWidth=0.1,NegativeWidth=0.1,PositiveHeight=0.1,NegativeHeight=1)
    ActiveRegionSize=0.1
    Debug=(bDrawDebug=false,DebugScale=25,DebugSegments=20,DebugThickness=0.25)
    DriverAxis=(X=-1,Y=0,Z=0)
    DriverItem=(Type=Control,Name="spine_05_ctrl")
    FalloffRegionScaleFactors=(PositiveWidth=1,NegativeWidth=1,PositiveHeight=1,NegativeHeight=2)
    FalloffSize=0.33
    FlipHeightScaling=false
    FlipWidthScaling=true
    OptionalParentItem=(Type=Bone,Name="spine_05")
    RotationOffset=(X=0,Y=30,Z=10)
   )
   > &RigUnit_ParentConstraint(
    AdvancedSettings=(RotationOrderForFilter=XZY)
    Child=(Type=Bone,Name="cape_01_r")
    Filter=(TranslationFilter=(bX=false,bY=false,bZ=false),RotationFilter=(bX=true,bY=true,bZ=true),ScaleFilter=(bX=True,bY=True,bZ=True))
    Parents=((Item=(Type=Bone,Name="thigh_twist_01_r"),Weight=0),(Item=(Type=Bone,Name="pelvis"),Weight=1))
    Weight=@n19.OutputParam
   )
   > &RigUnit_ModifyTransforms(
    ItemToModify="((Item=(Type=Bone,Name=\"cape_01_r\"),Transform=(Rotation=(X=-0.0227,Y=-0.1285,Z=-0.1722,W=0.9764),Translation=(X=0,Y=0,Z=0),Scale3D=(X=1,Y=1,Z=1))),(Item=(Type=Bone,Name=\"cape_mid_01_r\"),Transform=(Rotation=(X=-0.0227,Y=-0.1285,Z=-0.1722,W=0.9764),Translati ...(truncated)"
    Weight=@n19.OutputParam
   )
   > n20 = &RigUnit_SphericalPoseReader(
    ActiveRegionScaleFactors=(PositiveWidth=1,NegativeWidth=1,PositiveHeight=1.5,NegativeHeight=1)
    ActiveRegionSize=0.1
    Debug=(bDrawDebug=false,DebugScale=25,DebugSegments=20,DebugThickness=0.25)
    DriverAxis=(X=-1,Y=0,Z=0)
    DriverItem=(Type=Control,Name="spine_05_ctrl_2")
    FalloffRegionScaleFactors=(PositiveWidth=1,NegativeWidth=1,PositiveHeight=1,NegativeHeight=1)
    FalloffSize=0.3
    FlipHeightScaling=false
    FlipWidthScaling=false
    OptionalParentItem=(Type=Bone,Name="spine_05")
    RotationOffset=(X=180,Y=100,Z=-120)
   )
   > &RigUnit_ModifyTransforms(
    ItemToModify="((Item=(Type=Bone,Name=\"cape_02_r\"),Transform=(Rotation=(X=-0.067,Y=-0.25,Z=-0.25,W=0.933),Translation=(X=0,Y=0,Z=0),Scale3D=(X=1,Y=1,Z=1))),(Item=(Type=Bone,Name=\"cape_mid_02_r\"),Transform=(Rotation=(X=0.0151,Y=-0.173,Z=0.0858,W=0.9811),Translation=(X=0,Y ...(truncated)"
    Weight=@n20.OutputParam
   )
   n21 = RigVMDispatchNode(
    Array=&RigUnit_ItemArray(Items=((Type=Bone,Name="neck_01"),(Type=Bone,Name="neck_02"),(Type=Bone,Name="head"))).Items
    ExecuteContext=@SequenceAggregate.D
   ) {template="DISPATCH_RigVMDispatch_ArrayIterator(in Array,out Element,out Index,out Count,out Ratio)"}
   n22 = &RigUnit_GetTransform(Space=LocalSpace, bInitial=false, Item=@n21.Element)
   n23 = &RigVMFunction_MathQuaternionToEuler(RotationOrder=ZXY, Value=@n22.Transform.Rotation)
   > &RigUnit_SetTransform(
    Space=LocalSpace
    Value=(Rotation=(X=-0.0029,Y=0.7071,Z=0.0029,W=-0.7071),Translation=(X=0,Y=-1.6873,Z=152.0342),Scale3D=(X=1,Y=1,Z=1))
    Weight=1
    bInitial=False
    Item=@n21.Element
    Value.Rotation=&RigVMFunction_MathQuaternionFromEuler(
     Euler=(X=0,Y=0,Z=0)
     RotationOrder=ZXY
     Euler.X=&RigVMFunction_MathDoubleMul(B=1, A=@n23.Result.X).Result
     Euler.Z=&RigVMFunction_MathDoubleMul(B=0.5, A=@n23.Result.Z).Result
     Euler.Y=&RigVMFunction_MathDoubleMul(B=0.5, A=@n23.Result.Y).Result
    ).Result
    Value.Translation=@n22.Transform.Translation
    Value.Scale3D=@n22.Transform.Scale3D
   )
  ]}
]
hierarchy [
 root {loc=<0 0 0>, rot=<0 0 0 1>}
 pelvis {parent=root, loc=<0 -0.5301 98.0571>, rot=<0.0127 0.707 -0.0127 -0.707>}
 spine_01 {parent=pelvis, loc=<6.1208 -0.9328 0>, rot=<0 0 0.034 -0.9994>}
 spine_02 {parent=spine_01, loc=<6.5997 -0.2095 0>, rot=<0 0 0.0238 -0.9997>}
 spine_03 {parent=spine_02, loc=<6.4914 -0.1237 0>, rot=<0 0 0.0149 -0.9999>}
 spine_04 {parent=spine_03, loc=<6.9598 0.0794 0>, rot=<0 0 -0.0132 -0.9999>}
 spine_05 {parent=spine_04, loc=<6.9987 0 0>, rot=<0 0 -0.0015 -1>}
 attach_cape {parent=spine_05, loc=<9.017 -11.714 0>, rot=<0.0616 -0.7044 -0.0616 -0.7044>}
 cape {parent=attach_cape, type=Null, loc=<0 0 0>, rot=<0 0 0 1>}
 neck_01 {parent=spine_05, loc=<11.3138 0.1529 0>, rot=<0 0 -0.2003 -0.9797>}
 neck_02 {parent=neck_01, loc=<4.5627 0 0>, rot=<0 0 0.0504 -0.9987>}
 head {parent=neck_02, loc=<5.3914 0 0>, rot=<0 0 0.115 -0.9934>}
 attach_backpack {parent=spine_05, loc=<-0.123 -13.473 0>, rot=<-0.0093 -0.707 0.0093 -0.707>}
 backpack_root {parent=attach_backpack, loc=<0 10.9915 15.5853>, rot=<0.0532 0 0 -0.9986>}
 cape_mid_l_MSTR {parent=backpack_root, loc=<9.1529 1.3727 -1.6513>, rot=<-0.4534 -0.4475 0.539 -0.551>}
 cape_mid_01_l {parent=cape_mid_l_MSTR, loc=<3.792 0.0174 3.4704>, rot=<0.0144 0.0848 0.0049 -0.9963>}
 cape_mid_02_l {parent=cape_mid_01_l, loc=<24.3184 0.0016 -0.0001>, rot=<-0.0142 -0.1237 -0.0054 -0.9922>}
 cape_mid_03_l {parent=cape_mid_02_l, loc=<28.1116 0 -0.2127>, rot=<0 0.0023 0 -1>}
 cape_mid_04_l {parent=cape_mid_03_l, loc=<25.1969 0 0>, rot=<0 -0.0433 0 -0.9991>}
 cape_mid_05_l {parent=cape_mid_04_l, loc=<22.2027 0 0>, rot=<0 0 0 1>}
 cape_mid_06_l {parent=cape_mid_05_l, loc=<13.3811 0 0>, rot=<0 0 0 1>}
 cape_mid_r_MSTR {parent=backpack_root, loc=<-9.1529 1.3727 -1.6509>, rot=<-0.4475 -0.4534 0.551 -0.539>}
 cape_mid_01_r {parent=cape_mid_r_MSTR, loc=<3.7926 -0.0174 3.4702>, rot=<0.0144 0.0848 0.0049 -0.9963>}
 cape_mid_02_r {parent=cape_mid_01_r, loc=<24.316 0.3505 0.0101>, rot=<-0.0142 -0.1237 -0.0054 -0.9922>}
 cape_mid_03_r {parent=cape_mid_02_r, loc=<28.1112 0 -0.2125>, rot=<0 0.0023 0 -1>}
 cape_mid_04_r {parent=cape_mid_03_r, loc=<25.1969 0 -0.0001>, rot=<0 -0.0274 0 -0.9996>}
 cape_mid_05_r {parent=cape_mid_04_r, loc=<22.2027 0 0.0001>, rot=<0 -0.0223 0 -0.9998>}
 cape_mid_06_r {parent=cape_mid_05_r, loc=<13.3811 0 -0.0001>, rot=<0 0 0 1>}
 Backpack_BR {parent=attach_backpack, type=Null, loc=<0 5.2881 -1.6956>, rot=<0.089 0 0 0.996>}
 clavicle_l {parent=spine_05, loc=<5.4831 -0.1053 -3.2946>, rot=<0.0465 -0.7289 -0.009 -0.6829>}
 upperarm_l {parent=clavicle_l, loc=<9.8658 -1.2136 -0.4105>, rot=<0.0164 -0.3245 -0.0269 -0.9454>}
 lowerarm_l {parent=upperarm_l, loc=<26.465 -0.8345 0.259>, rot=<0.0036 0.0161 -0.2298 -0.9731>}
 hand_l {parent=lowerarm_l, loc=<22.7355 -0.3203 -0.6105>, rot=<-0.6784 -0.0223 0.0735 -0.7307>}
 thumb_01_l {parent=hand_l, loc=<1.3118 -0.3191 -2.459>, rot=<0.6546 -0.4042 0.0664 -0.6354>}
 thumb_02_l {parent=thumb_01_l, loc=<5.9336 -0.0448 0.0401>, rot=<-0.0037 0.0213 0.298 -0.9543>}
 thumb_03_l {parent=thumb_02_l, loc=<2.2223 0.3682 -0.0907>, rot=<0.0118 -0.013 0.0066 -0.9998>}
 index_metacarpal_l {parent=hand_l, loc=<1.6535 0.0183 -2.96>, rot=<0.0158 -0.0124 -0.0173 -0.9996>}
 index_01_l {parent=index_metacarpal_l, loc=<8.3008 0.0357 -0.0524>, rot=<-0.0044 0.0485 0.0924 -0.9945>}
 index_02_l {parent=index_01_l, loc=<3.3587 -0.3004 0.0452>, rot=<-0.002 0.0027 0.3225 -0.9466>}
 index_03_l {parent=index_02_l, loc=<2.5578 -0.123 -0.0252>, rot=<-0.002 0.0067 0.2558 -0.9667>}
 middle_metacarpal_l {parent=hand_l, loc=<2.1856 -0.1725 -0.5984>, rot=<0.006 0.0168 -0.0329 -0.9993>}
 middle_01_l {parent=middle_metacarpal_l, loc=<7.3275 -0.2447 0.0129>, rot=<-0.0208 0.0042 0.0858 -0.9961>}
 middle_02_l {parent=middle_01_l, loc=<4.0123 0.1401 0.1175>, rot=<-0.0218 0.0258 0.3512 -0.9357>}
 middle_03_l {parent=middle_02_l, loc=<2.6077 0.0534 0.1373>, rot=<0.0082 0.0188 0.1708 -0.9851>}
 ring_metacarpal_l {parent=hand_l, loc=<2.2074 -0.3625 1.3638>, rot=<-0.0284 0.0722 -0.0458 -0.9959>}
 ring_01_l {parent=ring_metacarpal_l, loc=<7.0295 -0.6197 -0.1446>, rot=<-0.0489 -0.0252 0.1892 -0.9804>}
 ring_02_l {parent=ring_01_l, loc=<3.6174 0.4647 -0.1027>, rot=<0 0 0.2806 -0.9598>}
 ring_03_l {parent=ring_02_l, loc=<2.4152 0.114 0.0199>, rot=<0 0 0.1189 -0.9929>}
 pinky_metacarpal_l {parent=hand_l, loc=<1.7616 -0.9753 3.3814>, rot=<-0.0397 0.0814 -0.069 -0.9935>}
 pinky_01_l {parent=pinky_metacarpal_l, loc=<7.1709 -0.4831 -0.1094>, rot=<-0.0654 -0.0035 0.2242 -0.9723>}
 pinky_02_l {parent=pinky_01_l, loc=<2.8274 0.0893 -0.1043>, rot=<0.0037 0.0011 0.2854 -0.9584>}
 pinky_03_l {parent=pinky_02_l, loc=<2.2046 0.0307 -0.0078>, rot=<-0.0002 0.0098 0.1445 -0.9895>}
 hand_attach_l {parent=hand_l, loc=<8.5 -3 0>, rot=<0.7071 0.7071 0 0>}
 weapon_l {parent=hand_l, loc=<-0.1974 -3.4111 0.3811>, rot=<0 0 0 1>}
 lowerarm_twist_02_l {parent=lowerarm_l, loc=<10.6707 0 0>, rot=<0 0 0 1>}
 lowerarm_twist_01_l {parent=lowerarm_l, loc=<18.316 0 0>, rot=<0 0 0 1>}
 deform_elbow_in_base_l {parent=lowerarm_l, loc=<-0.2322 5.6124 -0.0685>, rot=<0.0285 -0.0335 0.0024 -0.999>}
 deform_elbow_in_l {parent=deform_elbow_in_base_l, loc=<0 0 0>, rot=<0 0 0 1>}
 deform_elbow_out_base_l {parent=lowerarm_l, loc=<-0.5243 -3.6771 -0.3493>, rot=<0.0285 -0.0335 0.0024 -0.999>}
 deform_elbow_out_l {parent=deform_elbow_out_base_l, loc=<0 0 0>, rot=<0 0 0 1>}
 upperarm_twist_01_l {parent=upperarm_l, loc=<7.2135 -0.1134 0>, rot=<0 0 0 1>}
 deform_upperarm_bk_base_l {parent=upperarm_twist_01_l, loc=<-2.133 -6.2124 0.4116>, rot=<0 0.0007 0.0191 -0.9998>}
 deform_upperarm_bk_l {parent=deform_upperarm_bk_base_l, loc=<0 0 0>, rot=<0 0 0 1>}
 deform_upperarm_fr_base_l {parent=upperarm_twist_01_l, loc=<-3.036 8.1024 -0.4102>, rot=<0 0.0007 0.0191 -0.9998>}
 deform_upperarm_fr_l {parent=deform_upperarm_fr_base_l, loc=<0 0 0>, rot=<0 0 0 1>}
 deform_upperarm_up_base_l {parent=upperarm_twist_01_l, loc=<-7.0277 0.0163 6.5996>, rot=<0 0.0007 0.0191 -0.9998>}
 deform_upperarm_up_l {parent=deform_upperarm_up_base_l, loc=<0 0 0>, rot=<0 0 0 1>}
 upperarm_twist_02_l {parent=upperarm_l, loc=<19.7998 -0.1134 0>, rot=<0 0 0 1>}
 clavicle_cape_l {parent=clavicle_l, loc=<-0.8962 0.0005 -1.1875>, rot=<0 0 0 -1>}
 cape_l_MSTR {parent=clavicle_cape_l, loc=<13.5056 -1.2394 -0.0672>, rot=<-0.1932 -0.1531 0.6291 -0.7372>}
 cape_01_l {parent=cape_l_MSTR, loc=<9.8368 3.2935 4.0254>, rot=<-0.0383 -0.4151 0.0082 -0.9089>}
 cape_02_l {parent=cape_01_l, loc=<15.1546 0 -0.6849>, rot=<0 -0.0622 0 -0.9981>}
 cape_03_l {parent=cape_02_l, loc=<28.5209 0.275 -0.031>, rot=<0 0 0 1>}
 cape_04_l {parent=cape_03_l, loc=<23.8518 0 0>, rot=<0 -0.0068 0 -1>}
 cape_05_l {parent=cape_04_l, loc=<21.6657 0 0>, rot=<0 -0.0288 0 -0.9996>}
 cape_06_l {parent=cape_05_l, loc=<14.1389 0 0>, rot=<0 0 0 1>}
 cape_07_l {parent=cape_06_l, loc=<11.4453 0 0>, rot=<0 0 0 1>}
 dyn_shoulder_armor_l_MSTR {parent=clavicle_l, loc=<10.7493 -0.242 5.552>, rot=<0.0103 -0.3645 -0.0087 -0.9311>}
 dyn_shoulder_armor_l_Aim {parent=dyn_shoulder_armor_l_MSTR, loc=<0.2671 0 0>, rot=<0 0 0 1>}
 dyn_shoulder_armor_l_RBF {parent=dyn_shoulder_armor_l_Aim, loc=<0 0 0>, rot=<0 0 0 1>}
 dyn_shoulder_armor_l {parent=dyn_shoulder_armor_l_RBF, loc=<0 0 0>, rot=<0 0 0 1>}
 clavicle_r {parent=spine_05, loc=<5.4831 -0.1053 3.2946>, rot=<0.7289 0.0465 -0.6829 0.009>}
 upperarm_r {parent=clavicle_r, loc=<-9.8658 1.2136 0.4105>, rot=<0.0164 -0.3245 -0.0269 -0.9454>}
 lowerarm_r {parent=upperarm_r, loc=<-26.465 0.8345 -0.259>, rot=<0.0036 0.0161 -0.2298 -0.9731>}
 hand_r {parent=lowerarm_r, loc=<-22.7355 0.3203 0.6106>, rot=<-0.6784 -0.0223 0.0735 -0.7307>}
 thumb_01_r {parent=hand_r, loc=<-1.3118 0.3191 2.459>, rot=<0.6546 -0.4042 0.0664 -0.6354>}
 thumb_02_r {parent=thumb_01_r, loc=<-5.9336 0.0448 -0.0401>, rot=<-0.0037 0.0213 0.298 -0.9543>}
 thumb_03_r {parent=thumb_02_r, loc=<-2.2223 -0.3682 0.0907>, rot=<0.0118 -0.013 0.0066 -0.9998>}
 index_metacarpal_r {parent=hand_r, loc=<-1.6535 -0.0183 2.96>, rot=<0.0158 -0.0124 -0.0173 -0.9996>}
 index_01_r {parent=index_metacarpal_r, loc=<-8.3008 -0.0357 0.0524>, rot=<-0.0044 0.0485 0.0924 -0.9945>}
 index_02_r {parent=index_01_r, loc=<-3.3587 0.3004 -0.0452>, rot=<-0.002 0.0027 0.3225 -0.9466>}
 index_03_r {parent=index_02_r, loc=<-2.5578 0.123 0.0252>, rot=<-0.002 0.0067 0.2558 -0.9667>}
 middle_metacarpal_r {parent=hand_r, loc=<-2.1856 0.1725 0.5984>, rot=<0.006 0.0168 -0.0329 -0.9993>}
 middle_01_r {parent=middle_metacarpal_r, loc=<-7.3275 0.2447 -0.0129>, rot=<-0.0208 0.0042 0.0858 -0.9961>}
 middle_02_r {parent=middle_01_r, loc=<-4.0123 -0.1401 -0.1175>, rot=<-0.0218 0.0258 0.3512 -0.9357>}
 middle_03_r {parent=middle_02_r, loc=<-2.6077 -0.0534 -0.1373>, rot=<0.0082 0.0188 0.1708 -0.9851>}
 ring_metacarpal_r {parent=hand_r, loc=<-2.2074 0.3625 -1.3638>, rot=<-0.0284 0.0722 -0.0458 -0.9959>}
 ring_01_r {parent=ring_metacarpal_r, loc=<-7.0295 0.6197 0.1446>, rot=<-0.0489 -0.0252 0.1892 -0.9804>}
 ring_02_r {parent=ring_01_r, loc=<-3.6174 -0.4647 0.1027>, rot=<0 0 0.2806 -0.9598>}
 ring_03_r {parent=ring_02_r, loc=<-2.4152 -0.114 -0.0199>, rot=<0 0 0.1189 -0.9929>}
 pinky_metacarpal_r {parent=hand_r, loc=<-1.7616 0.9753 -3.3814>, rot=<-0.0397 0.0814 -0.069 -0.9935>}
 pinky_01_r {parent=pinky_metacarpal_r, loc=<-7.1709 0.4831 0.1094>, rot=<-0.0654 -0.0035 0.2242 -0.9723>}
 pinky_02_r {parent=pinky_01_r, loc=<-2.8274 -0.0893 0.1043>, rot=<0.0037 0.0011 0.2854 -0.9584>}
 pinky_03_r {parent=pinky_02_r, loc=<-2.2046 -0.0307 0.0078>, rot=<-0.0002 0.0098 0.1445 -0.9895>}
 weapon_r {parent=hand_r, loc=<0.1974 3.4111 -0.3811>, rot=<0 0 0 1>}
 hand_attach_r {parent=hand_r, loc=<-8.5001 2.9999 0>, rot=<0 0 -0.7071 -0.7071>}
 lowerarm_twist_02_r {parent=lowerarm_r, loc=<-10.6707 0 0>, rot=<0 0 0 -1>}
 lowerarm_twist_01_r {parent=lowerarm_r, loc=<-18.316 0 0>, rot=<0 0 0 -1>}
 deform_elbow_in_base_r {parent=lowerarm_r, loc=<0.2322 -5.6124 0.0685>, rot=<0.0285 -0.0335 0.0024 -0.999>}
 deform_elbow_in_r {parent=deform_elbow_in_base_r, loc=<-0.0001 0 0.0002>, rot=<0 0 0 1>}
 deform_elbow_out_base_r {parent=lowerarm_r, loc=<0.5243 3.6771 0.3493>, rot=<0.0285 -0.0335 0.0024 -0.999>}
 deform_elbow_out_r {parent=deform_elbow_out_base_r, loc=<0 0 0>, rot=<0 0 0 1>}
 upperarm_twist_01_r {parent=upperarm_r, loc=<-7.2135 0.1134 0>, rot=<0 0 0 -1>}
 deform_upperarm_bk_base_r {parent=upperarm_twist_01_r, loc=<2.1329 6.2124 -0.4114>, rot=<0 0.0007 0.0191 -0.9998>}
 deform_upperarm_bk_r {parent=deform_upperarm_bk_base_r, loc=<0 0 0>, rot=<0 0 0 1>}
 deform_upperarm_fr_base_r {parent=upperarm_twist_01_r, loc=<3.0358 -8.1024 0.4103>, rot=<0 0.0007 0.0191 -0.9998>}
 deform_upperarm_fr_r {parent=deform_upperarm_fr_base_r, loc=<0 0 0>, rot=<0 0 0 1>}
 deform_upperarm_up_base_r {parent=upperarm_twist_01_r, loc=<7.0274 -0.0163 -6.5994>, rot=<0 0.0007 0.0191 -0.9998>}
 deform_upperarm_up_r {parent=deform_upperarm_up_base_r, loc=<0 0 0>, rot=<0 0 0 -1>}
 upperarm_twist_02_r {parent=upperarm_r, loc=<-19.7998 0.1134 0>, rot=<0 0 0 -1>}
 clavicle_cape_r {parent=clavicle_r, loc=<0.8962 -0.0005 1.1875>, rot=<0 -1 0 0>}
 cape_r_MSTR {parent=clavicle_cape_r, loc=<13.5056 1.2394 -0.0671>, rot=<0.1932 -0.1531 -0.6291 -0.7372>}
 cape_01_r {parent=cape_r_MSTR, loc=<9.8097 -2.9782 4.0197>, rot=<0.0397 -0.4149 -0.0051 -0.909>}
 cape_02_r {parent=cape_01_r, loc=<15.1546 0 -0.6849>, rot=<0 -0.0622 0 -0.9981>}
 cape_03_r {parent=cape_02_r, loc=<28.5209 -0.275 -0.0311>, rot=<0 0 0 1>}
 cape_04_r {parent=cape_03_r, loc=<23.8517 0 0.0001>, rot=<0 -0.0346 0 -0.9994>}
 cape_05_r {parent=cape_04_r, loc=<21.6657 0 0>, rot=<0 -0.0346 0 -0.9994>}
 cape_06_r {parent=cape_05_r, loc=<14.1388 0 0>, rot=<0 0 0 1>}
 dyn_shoulder_armor_r_MSTR {parent=clavicle_r, loc=<-10.7493 0.242 -5.5517>, rot=<0.0103 -0.3645 -0.0087 -0.9311>}
 dyn_shoulder_armor_r_Aim {parent=dyn_shoulder_armor_r_MSTR, loc=<-0.2671 0 -0.0001>, rot=<0 0 0 1>}
 dyn_shoulder_armor_r_RBF {parent=dyn_shoulder_armor_r_Aim, loc=<0 0 0>, rot=<0 0 0 1>}
 dyn_shoulder_armor_r {parent=dyn_shoulder_armor_r_RBF, loc=<0 0 0>, rot=<0 0 0 1>}
 backpack_collision {parent=spine_05, loc=<-0.1896 -22.2129 0>, rot=<-0.0093 -0.707 0.0093 -0.707>}
 spine_05_ctrl {controlType=EULER_TRANSFORM, parent=spine_05, type=Control, loc=<0 0 0>, rot=<0 0 0 1>}
 spine_05_ctrl_2 {controlType=EULER_TRANSFORM, parent=spine_05, type=Control, loc=<0 0 0>, rot=<0 0 0 1>}
 dyn_chest_armor_fr_l_MSTR {parent=spine_05, loc=<1.6379 12.1787 -10.4968>, rot=<0.0399 -0.9954 0.0871 -0.0035>}
 dyn_chest_armor_fr_l {parent=dyn_chest_armor_fr_l_MSTR, loc=<0 0 0>, rot=<0 0 0 1>}
 dyn_chest_armor_fr_end_l {parent=dyn_chest_armor_fr_l, loc=<15.5607 0 0>, rot=<0 0 0 1>}
 dyn_chest_armor_bk_l_MSTR {parent=spine_05, loc=<5.616 -10.6194 -8.1473>, rot=<0 0 0.9992 -0.0401>}
 dyn_chest_armor_bk_l {parent=dyn_chest_armor_bk_l_MSTR, loc=<0 0 0>, rot=<0 0 0 1>}
 dyn_chest_armor_bk_end_l {parent=dyn_chest_armor_bk_l, loc=<17.0413 0 0>, rot=<0 0 0 1>}
 dyn_chest_armor_side_l_MSTR {parent=spine_05, loc=<-7.7675 -0.3038 -15.7515>, rot=<0.0284 -0.7065 0.7065 -0.0284>}
 dyn_chest_armor_side_l {parent=dyn_chest_armor_side_l_MSTR, loc=<0 0 0>, rot=<0 0 0 1>}
 dyn_chest_armor_side_end_l {parent=dyn_chest_armor_side_l, loc=<10.7825 0 0>, rot=<0 0 0 1>}
 dyn_chest_armor_fr_r_MSTR {parent=spine_05, loc=<1.6383 12.1788 10.4968>, rot=<0.9954 0.0399 -0.0035 -0.0871>}
 dyn_chest_armor_fr_r {parent=dyn_chest_armor_fr_r_MSTR, loc=<0 0 0>, rot=<0 0 0 1>}
 dyn_chest_armor_fr_end_r {parent=dyn_chest_armor_fr_r, loc=<-15.561 0 0>, rot=<0 0 0 -1>}
 dyn_chest_armor_side_r_MSTR {parent=spine_05, loc=<-7.7676 -0.3038 15.7515>, rot=<0.7065 0.0284 -0.0284 -0.7065>}
 dyn_chest_armor_side_r {parent=dyn_chest_armor_side_r_MSTR, loc=<0 0 0>, rot=<0 0 0 1>}
 dyn_chest_armor_side_end_r {parent=dyn_chest_armor_side_r, loc=<-10.782 0 0>, rot=<0 0 0 1>}
 dyn_chest_armor_bk_r_MSTR {parent=spine_05, loc=<5.616 -10.6193 8.1473>, rot=<0 0 -0.0401 -0.9992>}
 dyn_chest_armor_bk_r {parent=dyn_chest_armor_bk_r_MSTR, loc=<0 0 0>, rot=<0 0 0 1>}
 dyn_chest_armor_bk_end_r {parent=dyn_chest_armor_bk_r, loc=<-17.041 0 0>, rot=<0 0 0 1>}
 thigh_l {parent=pelvis, loc=<-1.0855 -0.2109 -8.7053>, rot=<0.0009 0.0355 0.0155 -0.9993>}
 calf_l {parent=thigh_l, loc=<-42.5284 -0.0019 -0.7297>, rot=<0.0006 -0.0161 -0.0391 -0.9991>}
 foot_l {parent=calf_l, loc=<-46.7401 -2.5014 -0.592>, rot=<0.0427 -0.0238 0.0353 -0.9982>}
 ball_l {parent=foot_l, loc=<-3.5658 10.8066 -0.1544>, rot=<0.0001 -0.0001 -0.7063 -0.7079>}
 foot_l_ctrl {controlType=EULER_TRANSFORM, parent=foot_l, type=Control, loc=<0 0 0>, rot=<0 0 0 1>}
 calf_twist_02_l {parent=calf_l, loc=<-19.1546 -1.7943 -0.2638>, rot=<0.0237 0.0095 -0.0035 -0.9997>}
 calf_twist_01_l {parent=calf_l, loc=<-35.8483 -1.9197 -0.5794>, rot=<0.0237 0.0095 -0.0035 -0.9997>}
 deform_knee_in_base_l {parent=calf_l, loc=<0.4409 -5.2372 0.2994>, rot=<0.0017 0.0216 0.0001 -0.9998>}
 deform_knee_in_l {parent=deform_knee_in_base_l, loc=<0 0 0>, rot=<0 0 0 1>}
 deform_knee_out_base_l {parent=calf_l, loc=<-0.8563 4.0587 0.2613>, rot=<0.0017 0.0216 0.0001 -0.9998>}
 deform_knee_out_l {parent=deform_knee_out_base_l, loc=<0 0 0>, rot=<0 0 0 1>}
 deform_knee_arm_out_l {parent=deform_knee_out_base_l, loc=<0.0721 -2.7571 -0.0587>, rot=<0 0 0 1>, scale=<1.4498 1 1.4498>}
 thigh_twist_01_l {parent=thigh_l, loc=<-27.952 0 0>, rot=<0.0013 0 0 -1>}
 deform_groin_base_l {parent=thigh_l, loc=<-8.4572 8.1142 -0.429>, rot=<0 0.0217 0 -0.9998>}
 deform_groin_l {parent=deform_groin_base_l, loc=<0 0 0>, rot=<0 0 0 1>}
 thigh_l_ctrl {controlType=EULER_TRANSFORM, parent=thigh_l, type=Control, loc=<0 0 0>, rot=<0 0 0 1>}
 thigh_r {parent=pelvis, loc=<-1.0855 -0.2109 8.7053>, rot=<0.0355 -0.0009 0.9993 0.0155>}
 calf_r {parent=thigh_r, loc=<42.5284 0.0019 0.7297>, rot=<0.0006 -0.0161 -0.0391 -0.9991>}
 foot_r {parent=calf_r, loc=<46.7401 2.5014 0.592>, rot=<0.0427 -0.0238 0.0353 -0.9982>}
 ball_r {parent=foot_r, loc=<3.5658 -10.8066 0.1544>, rot=<0.0001 -0.0001 -0.7063 -0.7079>}
 foot_r_ctrl {controlType=EULER_TRANSFORM, parent=foot_r, type=Control, loc=<0 0 0>, rot=<0 0 0 1>}
 calf_twist_02_r {parent=calf_r, loc=<19.1546 1.7943 0.2638>, rot=<0.0237 0.0095 -0.0035 -0.9997>}
 calf_twist_01_r {parent=calf_r, loc=<35.8483 1.9197 0.5794>, rot=<0.0237 0.0095 -0.0035 -0.9997>}
 deform_knee_in_base_r {parent=calf_r, loc=<-0.4409 5.2372 -0.2994>, rot=<0.0017 0.0216 0.0001 -0.9998>}
 deform_knee_in_r {parent=deform_knee_in_base_r, loc=<0 0 0>, rot=<0 0 0 1>}
 deform_knee_out_base_r {parent=calf_r, loc=<0.8564 -4.0587 -0.2613>, rot=<0.0017 0.0216 0.0001 -0.9998>}
 deform_knee_out_r {parent=deform_knee_out_base_r, loc=<0 0 0>, rot=<0 0 0 1>}
 deform_knee_arm_out_r {parent=deform_knee_out_base_r, loc=<-0.0722 2.7571 0.0587>, rot=<0.0315 0.7349 -0.027 -0.6769>}
 thigh_twist_01_r {parent=thigh_r, loc=<27.952 0 0>, rot=<0.0013 0 0 -1>}
 deform_groin_base_r {parent=thigh_r, loc=<8.4572 -8.1142 0.429>, rot=<0 0.0217 0 -0.9998>}
 deform_groin_r {parent=deform_groin_base_r, loc=<0 0 0>, rot=<0 0 0 1>}
 thigh_r_ctrl {controlType=EULER_TRANSFORM, parent=thigh_r, type=Control, loc=<0 0 0>, rot=<0 0 0 1>}
 deform_glute_base_l {parent=pelvis, loc=<-13.7475 -8.8222 -8.4287>, rot=<0.0275 0.0501 -0.4472 -0.8926>}
 deform_glute_l {parent=deform_glute_base_l, loc=<-0.0131 -0.0327 0>, rot=<0 0 0 1>}
 deform_glute_base_r {parent=pelvis, loc=<-13.7473 -8.8222 8.4287>, rot=<0.0501 -0.0275 0.8926 -0.4472>}
 deform_glute_r {parent=deform_glute_base_r, loc=<0.0131 0.0327 0>, rot=<0 0 0 1>}
 dyn_bag01 {parent=pelvis, loc=<0.2916 1.9323 -15.4808>, rot=<-0.014 -0.9839 -0.1546 -0.0889>}
 dyn_bag02 {parent=pelvis, loc=<1.8044 -6.8604 -15.5591>, rot=<0.0289 -0.9846 0.1575 -0.0696>}
 dyn_bag03 {parent=pelvis, loc=<0.3771 5.191 13.7477>, rot=<0.0548 0.2894 0.9556 0.0126>}
 dyn_bag04 {parent=pelvis, loc=<0.5848 -1.6424 16.2027>, rot=<0.0445 -0.0608 0.9972 0.0008>}
 gravity {parent=root, loc=<0 0 0>, rot=<0 0 0 1>}
 gravity_cloth {parent=root, loc=<0 0 0>, rot=<0 0 0 1>}
]
variables [
 spr_thigh_l_up_alpha {type=double, default=0}
 spr_thigh_r_up_alpha {type=double, default=0}
 spr_upperarm_l_up_alpha {type=double, default=0}
 spr_upperarm_r_up_alpha {type=double, default=0}
]
