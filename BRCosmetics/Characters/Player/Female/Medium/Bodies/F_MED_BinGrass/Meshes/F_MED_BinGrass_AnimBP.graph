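# Post-process style anim blueprint for the F_MED_BinGrass female body cosmetic, posing and simulating its cape, backpack and chest armor.
# The copied parent pose gets Control Rig, constraints and bone offsets, then cloth helper-bone offsets or state-selected RigidBody cape sims, chosen by bClothEnabled.
# Trail controllers add backward-movement drag, DBNO carry bypasses the simulation, and an emote flag adds cape rotations.
# The update event drives gravity override and cloth control through deprecated helpers.
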AnimBlueprint /BRCosmetics/Characters/Player/Female/Medium/Bodies/F_MED_BinGrass/Meshes/F_MED_BinGrass_AnimBP
target_skeleton=&/F_MED_BinGrass_Skeleton
graphs [
 AnimGraph {nodes [
   n1 = "Transform (Modify) Bone - Bone: cape_01_l"!(
    ComponentPose=Local To Component(LocalPose=Use cached pose 'PreSleeveSim'())
    bAlphaBoolEnabled:bool=True
    Translation="0,0,0"
    Rotation="10,0,0"
    Scale="1,1,1"
   ) -> (Pose) {
    alphaInputType=Bool
    rotationMode=BMM_Additive
    rotationSpace=BCS_BoneSpace
    translationMode=BMM_Additive
    translationSpace=BCS_BoneSpace
    alphaBoolBlend {blendOption=Linear}
    boneToModify {boneName=cape_01_l}
    rotation {pitch=10}
   }
   n2 = "Transform (Modify) Bone - Bone: cape_mid_01_l"!(ComponentPose=@n1, bAlphaBoolEnabled:bool=True, Translation="0,0,0", Rotation="10,0,0", Scale="1,1,1") -> (Pose) {
    alphaInputType=Bool
    rotationMode=BMM_Additive
    rotationSpace=BCS_BoneSpace
    translationMode=BMM_Additive
    translationSpace=BCS_BoneSpace
    alphaBoolBlend {blendOption=Linear}
    boneToModify {boneName=cape_mid_01_l}
    rotation {pitch=10}
   }
   n3 = "Transform (Modify) Bone - Bone: cape_01_r"!(ComponentPose=@n2, bAlphaBoolEnabled:bool=True, Translation="0,0,0", Rotation="10,0,0", Scale="1,1,1") -> (Pose) {
    alphaInputType=Bool
    rotationMode=BMM_Additive
    rotationSpace=BCS_BoneSpace
    translationMode=BMM_Additive
    translationSpace=BCS_BoneSpace
    alphaBoolBlend {blendOption=Linear}
    boneToModify {boneName=cape_01_r}
    rotation {pitch=10}
   }
   n4 = "Transform (Modify) Bone - Bone: cape_mid_01_r"!(ComponentPose=@n3, bAlphaBoolEnabled:bool=True, Translation="0,0,0", Rotation="10,0,0", Scale="1,1,1") -> (Pose) {
    alphaInputType=Bool
    rotationMode=BMM_Additive
    rotationSpace=BCS_BoneSpace
    translationMode=BMM_Additive
    translationSpace=BCS_BoneSpace
    alphaBoolBlend {blendOption=Linear}
    boneToModify {boneName=cape_mid_01_r}
    rotation {pitch=10}
   }
   n5 = "Transform (Modify) Bone - Bone: cape_02_l"!(ComponentPose=@n4, bAlphaBoolEnabled:bool=True, Translation="0,0,0", Rotation="20,0,0", Scale="1,1,1") -> (Pose) {
    alphaInputType=Bool
    rotationMode=BMM_Additive
    rotationSpace=BCS_BoneSpace
    translationMode=BMM_Additive
    translationSpace=BCS_BoneSpace
    alphaBoolBlend {blendOption=Linear}
    boneToModify {boneName=cape_02_l}
    rotation {pitch=20}
   }
   n6 = "Transform (Modify) Bone - Bone: cape_mid_02_l"!(ComponentPose=@n5, bAlphaBoolEnabled:bool=True, Translation="0,0,0", Rotation="20,0,0", Scale="1,1,1") -> (Pose) {
    alphaInputType=Bool
    rotationMode=BMM_Additive
    rotationSpace=BCS_BoneSpace
    translationMode=BMM_Additive
    translationSpace=BCS_BoneSpace
    alphaBoolBlend {blendOption=Linear}
    boneToModify {boneName=cape_mid_02_l}
    rotation {pitch=20}
   }
   n7 = "Transform (Modify) Bone - Bone: cape_02_r"!(ComponentPose=@n6, bAlphaBoolEnabled:bool=True, Translation="0,0,0", Rotation="20,0,0", Scale="1,1,1") -> (Pose) {
    alphaInputType=Bool
    rotationMode=BMM_Additive
    rotationSpace=BCS_BoneSpace
    translationMode=BMM_Additive
    translationSpace=BCS_BoneSpace
    alphaBoolBlend {blendOption=Linear}
    boneToModify {boneName=cape_02_r}
    rotation {pitch=20}
   }
   n8 = "Transform (Modify) Bone - Bone: cape_mid_02_r"!(ComponentPose=@n7, bAlphaBoolEnabled:bool=True, Translation="0,0,0", Rotation="20,0,0", Scale="1,1,1") -> (Pose) {
    alphaInputType=Bool
    rotationMode=BMM_Additive
    rotationSpace=BCS_BoneSpace
    translationMode=BMM_Additive
    translationSpace=BCS_BoneSpace
    alphaBoolBlend {blendOption=Linear}
    boneToModify {boneName=cape_mid_02_r}
    rotation {pitch=20}
   }
   n9 = "Transform (Modify) Bone - Bone: cape_03_l"!(ComponentPose=@n8, bAlphaBoolEnabled:bool=True, Translation="0,0,0", Rotation="15,0,0", Scale="1,1,1") -> (Pose) {
    alphaInputType=Bool
    rotationMode=BMM_Additive
    rotationSpace=BCS_BoneSpace
    translationMode=BMM_Additive
    translationSpace=BCS_BoneSpace
    alphaBoolBlend {blendOption=Linear}
    boneToModify {boneName=cape_03_l}
    rotation {pitch=15}
   }
   n10 = "Transform (Modify) Bone - Bone: cape_mid_03_l"!(ComponentPose=@n9, bAlphaBoolEnabled:bool=True, Translation="0,0,0", Rotation="15,0,0", Scale="1,1,1") -> (Pose) {
    alphaInputType=Bool
    rotationMode=BMM_Additive
    rotationSpace=BCS_BoneSpace
    translationMode=BMM_Additive
    translationSpace=BCS_BoneSpace
    alphaBoolBlend {blendOption=Linear}
    boneToModify {boneName=cape_mid_03_l}
    rotation {pitch=15}
   }
   n11 = "Transform (Modify) Bone - Bone: cape_03_r"!(ComponentPose=@n10, bAlphaBoolEnabled:bool=True, Translation="0,0,0", Rotation="15,0,0", Scale="1,1,1") -> (Pose) {
    alphaInputType=Bool
    rotationMode=BMM_Additive
    rotationSpace=BCS_BoneSpace
    translationMode=BMM_Additive
    translationSpace=BCS_BoneSpace
    alphaBoolBlend {blendOption=Linear}
    boneToModify {boneName=cape_03_r}
    rotation {pitch=15}
   }
   n12 = "Transform (Modify) Bone - Bone: cape_mid_03_r"!(ComponentPose=@n11, bAlphaBoolEnabled:bool=True, Translation="0,0,0", Rotation="15,0,0", Scale="1,1,1") -> (Pose) {
    alphaInputType=Bool
    rotationMode=BMM_Additive
    rotationSpace=BCS_BoneSpace
    translationMode=BMM_Additive
    translationSpace=BCS_BoneSpace
    alphaBoolBlend {blendOption=Linear}
    boneToModify {boneName=cape_mid_03_r}
    rotation {pitch=15}
   }
   n13 = "Transform (Modify) Bone - Bone: cape_03_l"(ComponentPose=@n12, bAlphaBoolEnabled=True, Translation="0,0,0", Rotation="20,0,0", Scale="1,1,1") {
    alphaInputType=Bool
    rotationMode=BMM_Additive
    rotationSpace=BCS_BoneSpace
    translationMode=BMM_Additive
    translationSpace=BCS_BoneSpace
    alphaBoolBlend {blendOption=Linear}
    boneToModify {boneName=cape_03_l}
    rotation {pitch=20}
   }
   n14 = "Transform (Modify) Bone - Bone: cape_mid_03_l"(ComponentPose=@n13, bAlphaBoolEnabled=True, Translation="0,0,0", Rotation="20,0,0", Scale="1,1,1") {
    alphaInputType=Bool
    rotationMode=BMM_Additive
    rotationSpace=BCS_BoneSpace
    translationMode=BMM_Additive
    translationSpace=BCS_BoneSpace
    alphaBoolBlend {blendOption=Linear}
    boneToModify {boneName=cape_mid_03_l}
    rotation {pitch=20}
   }
   n15 = "Transform (Modify) Bone - Bone: cape_03_r"(ComponentPose=@n14, bAlphaBoolEnabled=True, Translation="0,0,0", Rotation="20,0,0", Scale="1,1,1") {
    alphaInputType=Bool
    rotationMode=BMM_Additive
    rotationSpace=BCS_BoneSpace
    translationMode=BMM_Additive
    translationSpace=BCS_BoneSpace
    alphaBoolBlend {blendOption=Linear}
    boneToModify {boneName=cape_03_r}
    rotation {pitch=20}
   }
   n16 = "Transform (Modify) Bone - Bone: cape_mid_03_r"(ComponentPose=@n15, bAlphaBoolEnabled=True, Translation="0,0,0", Rotation="20,0,0", Scale="1,1,1") {
    alphaInputType=Bool
    rotationMode=BMM_Additive
    rotationSpace=BCS_BoneSpace
    translationMode=BMM_Additive
    translationSpace=BCS_BoneSpace
    alphaBoolBlend {blendOption=Linear}
    boneToModify {boneName=cape_mid_03_r}
    rotation {pitch=20}
   }
   Output Pose(Result=Blend Poses by bool(
     BlendPose_0=Component To Local(ComponentPose=@n16)
     BlendPose_1=Use cached pose 'PreSleeveSim'()
     BlendTime_1=0.7
     bActiveValue=Get emote_bingrass() -> (emote_bingrass:bool)
    ) {blendType=Linear}) {layerGroup=None}
   n17 = "Transform (Modify) Bone - Bone: backpack_collision"!(
    ComponentPose=Local To Component(LocalPose=Use cached pose 'standingPoseCache'())
    bAlphaBoolEnabled:bool=NOT Boolean(A=Get Is Wearing Backbling())
    Translation="0,0,300"
    Rotation="0,0,0"
    Scale="1,1,1"
   ) -> (Pose) {
    alphaInputType=Bool
    rotationMode=BMM_Additive
    rotationSpace=BCS_BoneSpace
    translationMode=BMM_Additive
    translationSpace=BCS_BoneSpace
    alphaBoolBlend {blendOption=Linear}
    boneToModify {boneName=backpack_collision}
    translation {z=300}
   }
   n18 = Component To Local(ComponentPose=@n17)
   n19 = "Transform (Modify) Bone - Bone: backpack_collision"(
    ComponentPose=Local To Component(LocalPose=Use cached pose 'standingPoseCache'())
    bAlphaBoolEnabled=NOT Boolean(A=Get Is Wearing Backbling())
    Translation="0,0,300"
    Rotation="0,0,0"
    Scale="1,1,1"
   ) {
    alphaInputType=Bool
    rotationMode=BMM_Additive
    rotationSpace=BCS_BoneSpace
    translationMode=BMM_Additive
    translationSpace=BCS_BoneSpace
    alphaBoolBlend {blendOption=Linear}
    boneToModify {boneName=backpack_collision}
    translation {z=300}
   }
   n20 = "Transform (Modify) Bone - Bone: gravity_cloth"!(ComponentPose=@n19, bAlphaBoolEnabled:bool=True, Translation="0,0,-100", Rotation="0,0,0", Scale="1,1,1") -> (Pose) {
    alphaInputType=Bool
    rotationMode=BMM_Additive
    rotationSpace=BCS_BoneSpace
    translationMode=BMM_Additive
    translationSpace=BCS_BoneSpace
    alphaBoolBlend {blendOption=Linear}
    boneToModify {boneName=gravity_cloth}
    translation {z=-100}
   }
   n21 = Component To Local(ComponentPose=@n20)
   n22 = Get Crouching or Emote or Siding() -> (Crouching or Emote or Siding:bool)
   n23 = RigidBody!(
    ComponentPose=Local To Component(LocalPose=Use cached pose 'standingPoseCache'())
    Alpha
    OverrideWorldGravity:Vector=Get Gravity Override Cape() -> (Gravity Override Cape:Vector)
    ExternalForce
    ComponentLinearAccScale:Vector="0,0,0"
    ComponentLinearVelScale:Vector="0,0,0"
    SimSpaceSettings:Sim Space Settings Structure=(WorldAlpha=1,VelocityScaleZ=1,DampingAlpha=1,MaxLinearVelocity=3000,MaxAngularVelocity=5,MaxLinearAcceleration=3000,MaxAngularAcceleration=3,ExternalLinearDragV=(X=0,Y=0,Z=0),ExternalLinearVelocity=(X=0,Y=0,Z=0),ExternalAngularVelocity=(X=0,Y=0,Z=0))
   ) -> (Pose) {
    bClampLinearTranslationLimitToRefPose=true
    bOverrideWorldGravity=true
    overridePhysicsAsset=&/F_MED_BinGrassBackBling_Cape_Physics2
    simulationSpace=BaseBoneSpace
    alphaBoolBlend {blendOption=Linear}
    baseBoneRef {boneName=root}
    overrideWorldGravity {z=-980}
    simSpaceSettings {maxAngularAcceleration=3, maxAngularVelocity=5, maxLinearAcceleration=3000, maxLinearVelocity=3000, worldAlpha=1}
   }
   n24 = RigidBody(
    ComponentPose=Local To Component(LocalPose=Use cached pose 'standingPoseCache'())
    OverrideWorldGravity=Get Gravity Override Cape()
    ComponentLinearAccScale="0,0,0"
    ComponentLinearVelScale="0,0,0"
    SimSpaceSettings=(WorldAlpha=1,VelocityScaleZ=1,DampingAlpha=1,MaxLinearVelocity=3000,MaxAngularVelocity=5,MaxLinearAcceleration=3000,MaxAngularAcceleration=3,ExternalLinearDragV=(X=0,Y=0,Z=0),ExternalLinearVelocity=(X=0,Y=0,Z=0),ExternalAngularVelocity=(X=0,Y=0,Z=0))
   ) {
    bClampLinearTranslationLimitToRefPose=true
    bOverrideWorldGravity=true
    overridePhysicsAsset=&/F_MED_BinGrass_Cape_Physics
    simulationSpace=BaseBoneSpace
    alphaBoolBlend {blendOption=Linear}
    baseBoneRef {boneName=root}
    overrideWorldGravity {z=-980}
    simSpaceSettings {maxAngularAcceleration=3, maxAngularVelocity=5, maxLinearAcceleration=3000, maxLinearVelocity=3000, worldAlpha=1}
   }
   n25 = Blend Poses by bool(
    BlendPose_0=Blend Poses by bool(
     BlendPose_0=Component To Local(ComponentPose=@n23)
     BlendPose_1=Component To Local(ComponentPose=@n24)
     bActiveValue=Get Is Wearing Backbling()
    ) {blendType=Linear}
    BlendPose_1=Use cached pose 'posBackbing'()
    bActiveValue=Get FrontEnd or Emote() -> (FrontEnd or Emote:bool)
   ) {blendType=Linear}
   n26 = RigidBody!(
    ComponentPose=Local To Component(LocalPose=Use cached pose 'posBackbing'())
    Alpha=Get in vehicle rigid alpha() -> (in vehicle rigid alpha:double)
    OverrideWorldGravity=Get Gravity Override Cape()
    ExternalForce
    SimSpaceSettings=(WorldAlpha=0,VelocityScaleZ=0,DampingAlpha=1,MaxLinearVelocity=150,MaxAngularVelocity=1,MaxLinearAcceleration=300,MaxAngularAcceleration=50,ExternalLinearDragV=(X=0.2,Y=0.2,Z=0),ExternalLinearVelocity=(X=0,Y=0,Z=0),ExternalAngularVelocity=(X=0,Y=0,Z=0))
   ) -> (Pose) {
    bClampLinearTranslationLimitToRefPose=true
    bOverrideWorldGravity=true
    overridePhysicsAsset=&F_MED_BinGrassBackBling_Cape_Physics2
    simulationSpace=BaseBoneSpace
    alphaBoolBlend {blendOption=Linear}
    baseBoneRef {boneName=gravity}
    overrideWorldGravity {z=-980}
    simSpaceSettings {
     maxAngularAcceleration=50
     maxAngularVelocity=1
     maxLinearAcceleration=300
     maxLinearVelocity=150
     velocityScaleZ=0
     externalLinearDragV=<0.2 0.2>
    }
   }
   n27 = RigidBody!(
    ComponentPose=Local To Component(LocalPose=Use cached pose 'standingPoseCache'())
    Alpha
    OverrideWorldGravity=Get Gravity Override Cape()
    ExternalForce
    ComponentLinearAccScale:Vector="0,0,0"
    ComponentLinearVelScale:Vector="0,0,0"
    SimSpaceSettings=(WorldAlpha=1,VelocityScaleZ=1,DampingAlpha=1,MaxLinearVelocity=3000,MaxAngularVelocity=5,MaxLinearAcceleration=3000,MaxAngularAcceleration=3,ExternalLinearDragV=(X=0,Y=0,Z=0),ExternalLinearVelocity=(X=0,Y=0,Z=0),ExternalAngularVelocity=(X=0,Y=0,Z=0))
   ) -> (Pose) {
    bClampLinearTranslationLimitToRefPose=true
    bOverrideWorldGravity=true
    overridePhysicsAsset=&F_MED_BinGrass_Cape_Physics
    simulationSpace=BaseBoneSpace
    alphaBoolBlend {blendOption=Linear}
    baseBoneRef {boneName=pelvis}
    overrideWorldGravity {z=-980}
    simSpaceSettings {maxAngularAcceleration=3, maxAngularVelocity=5, maxLinearAcceleration=3000, maxLinearVelocity=3000, worldAlpha=1}
   }
   n28 = Blend Poses by bool(
    BlendPose_0=Component To Local(ComponentPose=@n26)
    BlendPose_1=Component To Local(ComponentPose=@n27)
    bActiveValue=Get Is Wearing Backbling()
   ) {blendType=Linear}
   n29 = Get Crouching or Emote or Siding()
   n30 = "Transform (Modify) Bone - Bone: cape_mid_01_l"(
    ComponentPose=Local To Component(LocalPose=Blend Poses by bool(BlendPose_0=@n28, BlendPose_1=@n25, BlendTime_1=0.7, bActiveValue=@n29) {blendType=Linear})
    bAlphaBoolEnabled=True
    Translation="0,0,-4"
    Rotation="5,0,0"
    Scale="1,1,1"
   ) {
    alphaInputType=Bool
    rotationMode=BMM_Additive
    rotationSpace=BCS_BoneSpace
    translationMode=BMM_Additive
    translationSpace=BCS_BoneSpace
    alphaBoolBlend {blendOption=Linear}
    boneToModify {boneName=cape_mid_01_l}
    rotation {pitch=5}
    translation {z=-4}
   }
   n31 = "Transform (Modify) Bone - Bone: cape_mid_01_r"(ComponentPose=@n30, bAlphaBoolEnabled=True, Translation="0,0,-4", Rotation="5,0,0", Scale="1,1,1") {
    alphaInputType=Bool
    rotationMode=BMM_Additive
    rotationSpace=BCS_BoneSpace
    translationMode=BMM_Additive
    translationSpace=BCS_BoneSpace
    alphaBoolBlend {blendOption=Linear}
    boneToModify {boneName=cape_mid_01_r}
    rotation {pitch=5}
    translation {z=-4}
   }
   n32 = "Transform (Modify) Bone - Bone: cape_01_l"(ComponentPose=@n31, bAlphaBoolEnabled=True, Translation="0,0,-4", Rotation="5,0,0", Scale="1,1,1") {
    alphaInputType=Bool
    rotationMode=BMM_Additive
    rotationSpace=BCS_BoneSpace
    translationMode=BMM_Additive
    translationSpace=BCS_BoneSpace
    alphaBoolBlend {blendOption=Linear}
    boneToModify {boneName=cape_01_l}
    rotation {pitch=5}
    translation {z=-4}
   }
   n33 = "Transform (Modify) Bone - Bone: cape_01_r"(ComponentPose=@n32, bAlphaBoolEnabled=True, Translation="0,0,-4", Rotation="5,0,0", Scale="1,1,1") {
    alphaInputType=Bool
    rotationMode=BMM_Additive
    rotationSpace=BCS_BoneSpace
    translationMode=BMM_Additive
    translationSpace=BCS_BoneSpace
    alphaBoolBlend {blendOption=Linear}
    boneToModify {boneName=cape_01_r}
    rotation {pitch=5}
    translation {z=-4}
   }
   Save cached pose 'postSim'(Pose=Blend Poses by bool(
     BlendPose_0=Blend Poses by bool(BlendPose_0=@n18, BlendPose_1=@n21, BlendTime_0=0.2, BlendTime_1=0.5, bActiveValue=@n22) {blendType=Linear}
     BlendPose_1=Component To Local(ComponentPose=@n33)
     bActiveValue=Get bClothEnabled()
    ) {blendType=Linear}) {cacheName=postSim}
   n34 = "Transform (Modify) Bone - Bone: cape_01_l"(
    ComponentPose=Local To Component(LocalPose=Use cached pose 'postSim'())
    bAlphaBoolEnabled=Get bIsInFrontEnd()
    Translation="-3,0,0"
    Rotation="0,0,0"
    Scale="1,1,1"
   ) {
    alphaInputType=Bool
    rotationMode=BMM_Additive
    scaleMode=BMM_Additive
    translationMode=BMM_Additive
    alphaBoolBlend {blendOption=Linear}
    boneToModify {boneName=cape_01_l}
    translation {x=-3}
   }
   Save cached pose 'RBANCache'(Pose=Component To Local(ComponentPose=@n34)) {cacheName=RBANCache}
   n35 = "Transform (Modify) Bone - Bone: attach_backpack"!(
    ComponentPose=Local To Component(LocalPose=Copy Pose From Mesh() {bCopyCurves=true, bUseAttachedParent=true})
    bAlphaBoolEnabled:bool=True
    Translation="0,-2,0"
    Rotation="0,0,25"
    Scale="1,1,1"
   ) -> (Pose) {
    alphaInputType=Bool
    rotationMode=BMM_Additive
    rotationSpace=BCS_BoneSpace
    translationMode=BMM_Additive
    translationSpace=BCS_BoneSpace
    alphaBoolBlend {blendOption=Linear}
    boneToModify {boneName=attach_backpack}
    rotation {roll=25}
    translation {y=-2}
   }
   n36 = "Transform (Modify) Bone - Bone: backpack_root"!(ComponentPose=@n35, bAlphaBoolEnabled:bool=True, Translation="0,0,0", Rotation="0,0,-20", Scale="1,1,1") -> (Pose) {
    alphaInputType=Bool
    rotationMode=BMM_Additive
    rotationSpace=BCS_BoneSpace
    translationMode=BMM_Additive
    translationSpace=BCS_BoneSpace
    alphaBoolBlend {blendOption=Linear}
    boneToModify {boneName=backpack_root}
    rotation {roll=-20}
   }
   n37 = Control Rig(Source=Component To Local(ComponentPose=@n36)) {
    alphaBoolBlend {blendOption=Linear}
    controlRigAssetReference {blueprintRigClass=&/M_MED_SherWolf_Gnash_CtrlRig.M_MED_SherWolf_Gnash_CtrlRig_C}
   }
   n38 = "Constraint - backpack_root (P)"!(ComponentPose=Local To Component(LocalPose=@n37), bAlphaBoolEnabled:bool=Get Is Wearing Backbling(), ConstraintWeights_0=1) -> (Pose) {
    alphaInputType=Bool
    alphaBoolBlend {blendOption=Linear}
    boneToModify {boneName=backpack_root}
    constraintSetup [{targetBone {boneName=spine_05}}]
    constraintWeights [1]
   }
   n39 = "Constraint - gravity (TT)"!(ComponentPose=@n38, bAlphaBoolEnabled:bool=True, ConstraintWeights_0=0.2, ConstraintWeights_1:float=0.8) -> (Pose) {
    alphaInputType=Bool
    alphaBoolBlend {blendOption=Linear}
    boneToModify {boneName=gravity}
    constraintSetup [{transformType=Translation, targetBone {boneName=root}}, {transformType=Translation, targetBone {boneName=spine_03}}]
    constraintWeights [1,1]
   }
   n40 = "Transform (Modify) Bone - Bone: cape_mid_03_r"(ComponentPose=@n39, bAlphaBoolEnabled=True, Translation="0,0,0", Rotation="-20,0,0", Scale="1,1,1") {
    alphaInputType=Bool
    rotationMode=BMM_Additive
    rotationSpace=BCS_BoneSpace
    translationMode=BMM_Additive
    translationSpace=BCS_BoneSpace
    alphaBoolBlend {blendOption=Linear}
    boneToModify {boneName=cape_mid_03_r}
    rotation {pitch=-20}
   }
   n41 = "Transform (Modify) Bone - Bone: cape_01_l"(ComponentPose=@n40, bAlphaBoolEnabled=True, Translation="0,0,0", Rotation="-30,0,0", Scale="1,1,1") {
    alphaInputType=Bool
    rotationMode=BMM_Additive
    rotationSpace=BCS_BoneSpace
    translationMode=BMM_Additive
    translationSpace=BCS_BoneSpace
    alphaBoolBlend {blendOption=Linear}
    boneToModify {boneName=cape_01_l}
    rotation {pitch=-30}
   }
   n42 = "Transform (Modify) Bone - Bone: cape_mid_03_l"(ComponentPose=@n41, bAlphaBoolEnabled=True, Translation="0,0,0", Rotation="-20,0,0", Scale="1,1,1") {
    alphaInputType=Bool
    rotationMode=BMM_Additive
    rotationSpace=BCS_BoneSpace
    translationMode=BMM_Additive
    translationSpace=BCS_BoneSpace
    alphaBoolBlend {blendOption=Linear}
    boneToModify {boneName=cape_mid_03_l}
    rotation {pitch=-20}
   }
   n43 = "Transform (Modify) Bone - Bone: cape_01_r"(ComponentPose=@n42, bAlphaBoolEnabled=True, Translation="0,0,0", Rotation="-30,0,0", Scale="1,1,1") {
    alphaInputType=Bool
    rotationMode=BMM_Additive
    rotationSpace=BCS_BoneSpace
    translationMode=BMM_Additive
    translationSpace=BCS_BoneSpace
    alphaBoolBlend {blendOption=Linear}
    boneToModify {boneName=cape_01_r}
    rotation {pitch=-30}
   }
   n44 = "Transform (Modify) Bone - Bone: cape_mid_04_r"!(ComponentPose=@n43, bAlphaBoolEnabled:bool=True, Translation="0,0,0", Rotation="-20,0,0", Scale="1,1,1") -> (Pose) {
    alphaInputType=Bool
    rotationMode=BMM_Additive
    rotationSpace=BCS_BoneSpace
    translationMode=BMM_Additive
    translationSpace=BCS_BoneSpace
    alphaBoolBlend {blendOption=Linear}
    boneToModify {boneName=cape_mid_04_r}
    rotation {pitch=-20}
   }
   n45 = "Transform (Modify) Bone - Bone: cape_04_l"!(ComponentPose=@n44, bAlphaBoolEnabled:bool=True, Translation="0,0,0", Rotation="-20,0,0", Scale="1,1,1") -> (Pose) {
    alphaInputType=Bool
    rotationMode=BMM_Additive
    rotationSpace=BCS_BoneSpace
    translationMode=BMM_Additive
    translationSpace=BCS_BoneSpace
    alphaBoolBlend {blendOption=Linear}
    boneToModify {boneName=cape_04_l}
    rotation {pitch=-20}
   }
   n46 = "Transform (Modify) Bone - Bone: cape_mid_04_l"!(ComponentPose=@n45, bAlphaBoolEnabled:bool=True, Translation="0,0,0", Rotation="-20,0,0", Scale="1,1,1") -> (Pose) {
    alphaInputType=Bool
    rotationMode=BMM_Additive
    rotationSpace=BCS_BoneSpace
    translationMode=BMM_Additive
    translationSpace=BCS_BoneSpace
    alphaBoolBlend {blendOption=Linear}
    boneToModify {boneName=cape_mid_04_l}
    rotation {pitch=-20}
   }
   n47 = "Transform (Modify) Bone - Bone: cape_04_r"!(ComponentPose=@n46, bAlphaBoolEnabled:bool=True, Translation="0,0,0", Rotation="-20,0,0", Scale="1,1,1") -> (Pose) {
    alphaInputType=Bool
    rotationMode=BMM_Additive
    rotationSpace=BCS_BoneSpace
    translationMode=BMM_Additive
    translationSpace=BCS_BoneSpace
    alphaBoolBlend {blendOption=Linear}
    boneToModify {boneName=cape_04_r}
    rotation {pitch=-20}
   }
   n48 = "Transform (Modify) Bone - Bone: cape_mid_03_r"(ComponentPose=@n47, bAlphaBoolEnabled=True, Translation="0,0,0", Rotation="-30,0,0", Scale="1,1,1") {
    alphaInputType=Bool
    rotationMode=BMM_Additive
    rotationSpace=BCS_BoneSpace
    translationMode=BMM_Additive
    translationSpace=BCS_BoneSpace
    alphaBoolBlend {blendOption=Linear}
    boneToModify {boneName=cape_mid_03_r}
    rotation {pitch=-30}
   }
   n49 = "Transform (Modify) Bone - Bone: cape_02_l"(ComponentPose=@n48, bAlphaBoolEnabled=True, Translation="0,0,0", Rotation="-20,30,0", Scale="1,1,1") {
    alphaInputType=Bool
    rotationMode=BMM_Additive
    rotationSpace=BCS_BoneSpace
    translationMode=BMM_Additive
    translationSpace=BCS_BoneSpace
    alphaBoolBlend {blendOption=Linear}
    boneToModify {boneName=cape_02_l}
    rotation {pitch=-20, yaw=30}
   }
   n50 = "Transform (Modify) Bone - Bone: cape_mid_03_l"(ComponentPose=@n49, bAlphaBoolEnabled=True, Translation="0,0,0", Rotation="-30,0,0", Scale="1,1,1") {
    alphaInputType=Bool
    rotationMode=BMM_Additive
    rotationSpace=BCS_BoneSpace
    translationMode=BMM_Additive
    translationSpace=BCS_BoneSpace
    alphaBoolBlend {blendOption=Linear}
    boneToModify {boneName=cape_mid_03_l}
    rotation {pitch=-30}
   }
   n51 = "Transform (Modify) Bone - Bone: cape_02_r"(ComponentPose=@n50, bAlphaBoolEnabled=True, Translation="0,0,0", Rotation="-20,0,0", Scale="1,1,1") {
    alphaInputType=Bool
    rotationMode=BMM_Additive
    rotationSpace=BCS_BoneSpace
    translationMode=BMM_Additive
    translationSpace=BCS_BoneSpace
    alphaBoolBlend {blendOption=Linear}
    boneToModify {boneName=cape_02_r}
    rotation {pitch=-20}
   }
   n52 = "Transform (Modify) Bone - Bone: dyn_bag03"!(ComponentPose=@n51, bAlphaBoolEnabled:bool=True, Translation="0,0,1", Rotation="10,0,0", Scale="1,1,1") -> (Pose) {
    alphaInputType=Bool
    rotationMode=BMM_Additive
    rotationSpace=BCS_BoneSpace
    translationMode=BMM_Additive
    translationSpace=BCS_BoneSpace
    alphaBoolBlend {blendOption=Linear}
    boneToModify {boneName=dyn_bag03}
    rotation {pitch=10}
    translation {z=1}
   }
   n53 = "Constraint - dyn_chest_armor_bk_l_MSTR (PPP)"!(ComponentPose=@n52, Alpha=1, ConstraintWeights_0=0.3, ConstraintWeights_1:float=0.5, ConstraintWeights_2:float=0.2) -> (Pose) {
    alphaBoolBlend {blendOption=Linear}
    boneToModify {boneName=dyn_chest_armor_bk_l_MSTR}
    constraintSetup [{targetBone {boneName=spine_04}}, {targetBone {boneName=spine_03}}, {targetBone {boneName=spine_05}}]
    constraintWeights [1,1,1]
   }
   n54 = "Constraint - dyn_chest_armor_bk_r_MSTR (PPP)"!(ComponentPose=@n53, Alpha=1, ConstraintWeights_0=0.3, ConstraintWeights_1:float=0.5, ConstraintWeights_2:float=0.2) -> (Pose) {
    alphaBoolBlend {blendOption=Linear}
    boneToModify {boneName=dyn_chest_armor_bk_r_MSTR}
    constraintSetup [{targetBone {boneName=spine_04}}, {targetBone {boneName=spine_03}}, {targetBone {boneName=spine_05}}]
    constraintWeights [1,1,1]
   }
   n55 = "Constraint - dyn_chest_armor_side_l_MSTR (PPP)"!(ComponentPose=@n54, Alpha=1, ConstraintWeights_0=0.5, ConstraintWeights_1:float=0.5, ConstraintWeights_2:float=0) -> (Pose) {
    alphaBoolBlend {blendOption=Linear}
    boneToModify {boneName=dyn_chest_armor_side_l_MSTR}
    constraintSetup [{targetBone {boneName=spine_04}}, {targetBone {boneName=spine_03}}, {targetBone {boneName=spine_05}}]
    constraintWeights [1,1,1]
   }
   n56 = "Constraint - dyn_chest_armor_side_r_MSTR (PPP)"!(ComponentPose=@n55, Alpha=1, ConstraintWeights_0=0.5, ConstraintWeights_1:float=0.5, ConstraintWeights_2:float=0) -> (Pose) {
    alphaBoolBlend {blendOption=Linear}
    boneToModify {boneName=dyn_chest_armor_side_r_MSTR}
    constraintSetup [{targetBone {boneName=spine_04}}, {targetBone {boneName=spine_03}}, {targetBone {boneName=spine_05}}]
    constraintWeights [1,1,1]
   }
   n57 = "Constraint - dyn_chest_armor_fr_l_MSTR (PPP)"!(ComponentPose=@n56, Alpha=1, ConstraintWeights_0=0.5, ConstraintWeights_1:float=0.5, ConstraintWeights_2:float=0) -> (Pose) {
    alphaBoolBlend {blendOption=Linear}
    boneToModify {boneName=dyn_chest_armor_fr_l_MSTR}
    constraintSetup [{targetBone {boneName=spine_04}}, {targetBone {boneName=spine_03}}, {targetBone {boneName=spine_05}}]
    constraintWeights [1,1,1]
   }
   n58 = "Constraint - dyn_chest_armor_fr_r_MSTR (PPP)"!(ComponentPose=@n57, Alpha=1, ConstraintWeights_0=0.5, ConstraintWeights_1:float=0.5, ConstraintWeights_2:float=0) -> (Pose) {
    alphaBoolBlend {blendOption=Linear}
    boneToModify {boneName=dyn_chest_armor_fr_r_MSTR}
    constraintSetup [{targetBone {boneName=spine_04}}, {targetBone {boneName=spine_03}}, {targetBone {boneName=spine_05}}]
    constraintWeights [1,1,1]
   }
   n59 = "Constraint - dyn_chest_armor_bk_end_l (PPP)"!(ComponentPose=@n58, Alpha=1, ConstraintWeights_0=0.3, ConstraintWeights_1:float=0.5, ConstraintWeights_2:float=0.2) -> (Pose) {
    alphaBoolBlend {blendOption=Linear}
    boneToModify {boneName=dyn_chest_armor_bk_end_l}
    constraintSetup [{targetBone {boneName=spine_04}}, {targetBone {boneName=spine_02}}, {targetBone {boneName=spine_05}}]
    constraintWeights [1,1,1]
   }
   n60 = "Constraint - dyn_chest_armor_bk_end_r (PPP)"!(ComponentPose=@n59, Alpha=1, ConstraintWeights_0=0.3, ConstraintWeights_1:float=0.5, ConstraintWeights_2:float=0.2) -> (Pose) {
    alphaBoolBlend {blendOption=Linear}
    boneToModify {boneName=dyn_chest_armor_bk_end_r}
    constraintSetup [{targetBone {boneName=spine_04}}, {targetBone {boneName=spine_02}}, {targetBone {boneName=spine_05}}]
    constraintWeights [1,1,1]
   }
   n61 = "Transform (Modify) Bone - Bone: dyn_chest_armor_bk_l_MSTR"(ComponentPose=@n60, Alpha=1, Translation="0,0,0", Rotation="0,5,0", Scale="1,1,1") {
    rotationMode=BMM_Additive
    rotationSpace=BCS_BoneSpace
    translationMode=BMM_Additive
    translationSpace=BCS_BoneSpace
    alphaBoolBlend {blendOption=Linear}
    alphaScaleBiasClamp {inRange {max=0, min=-100}, outRange {max=0, min=1}}
    boneToModify {boneName=dyn_chest_armor_bk_l_MSTR}
    rotation {yaw=5}
   }
   n62 = "Transform (Modify) Bone - Bone: dyn_chest_armor_bk_r_MSTR"(ComponentPose=@n61, Alpha=1, Translation="0,0,0", Rotation="0,5,0", Scale="1,1,1") {
    rotationMode=BMM_Additive
    rotationSpace=BCS_BoneSpace
    translationMode=BMM_Additive
    translationSpace=BCS_BoneSpace
    alphaBoolBlend {blendOption=Linear}
    alphaScaleBiasClamp {inRange {max=0, min=-100}, outRange {max=0, min=1}}
    boneToModify {boneName=dyn_chest_armor_bk_r_MSTR}
    rotation {yaw=5}
   }
   n63 = "Transform (Modify) Bone - Bone: dyn_chest_armor_side_l_MSTR"(ComponentPose=@n62, Alpha=1, Translation="0,0,0", Rotation="0,8,0", Scale="1,1,1") {
    rotationMode=BMM_Additive
    rotationSpace=BCS_BoneSpace
    translationMode=BMM_Additive
    translationSpace=BCS_BoneSpace
    alphaBoolBlend {blendOption=Linear}
    alphaScaleBiasClamp {inRange {max=0, min=-100}, outRange {max=0, min=1}}
    boneToModify {boneName=dyn_chest_armor_side_l_MSTR}
    rotation {yaw=8}
   }
   n64 = "Transform (Modify) Bone - Bone: dyn_chest_armor_side_r_MSTR"(ComponentPose=@n63, Alpha=1, Translation="0,0,0", Rotation="0,8,0", Scale="1,1,1") {
    rotationMode=BMM_Additive
    rotationSpace=BCS_BoneSpace
    translationMode=BMM_Additive
    translationSpace=BCS_BoneSpace
    alphaBoolBlend {blendOption=Linear}
    alphaScaleBiasClamp {inRange {max=0, min=-100}, outRange {max=0, min=1}}
    boneToModify {boneName=dyn_chest_armor_side_r_MSTR}
    rotation {yaw=8}
   }
   Save cached pose 'standingPoseCache'(Pose=Component To Local(ComponentPose=@n64)) {cacheName=standingPoseCache}
   n65 = Get BwdMovement() -> (BwdMovement:double)
   n66 = Trail controller - Bone: cape_mid_06_l(ComponentPose=Local To Component(LocalPose=Use cached pose 'RBANCache'()), Alpha=@n65) {
    bUsePlanarLimit=true
    chainLength=5
    debugLifeTime=0.1
    maxDeltaTime=0.0333
    alphaScaleBiasClamp {bInterpResult=true}
    rotationLimits [
     {limitMax=<180 180 180>}
     {limitMax=<180 180 180>}
     {limitMax=<180 90 180>, limitMin=<-180 -90 -180>}
     {limitMax=<180 180 180>}
     {limitMax=<180 180 180>}
    ]
    rotationOffsets [<0 0 0>, <0 0 0>, <0 0 0>, <0 0 0>, <0 0 0>]
    trailBone {boneName=cape_mid_06_l}
    trailRelaxationSpeed {editorCurveData {keys [<at 0 15 RCIM_Linear>,<at 0.5289 13.1838 RCIM_Linear>,<at 1 6 RCIM_Linear>]}}
   }
   n67 = Trail controller - Bone: cape_mid_06_r(ComponentPose=@n66, Alpha=@n65) {
    bUsePlanarLimit=true
    chainLength=5
    debugLifeTime=0.1
    maxDeltaTime=0.0333
    alphaScaleBiasClamp {bInterpResult=true}
    rotationLimits [
     {limitMax=<180 180 180>}
     {limitMax=<180 180 180>}
     {limitMax=<180 90 180>, limitMin=<-180 -90 -180>}
     {limitMax=<180 180 180>}
     {limitMax=<180 180 180>}
    ]
    rotationOffsets [<0 0 0>, <0 0 0>, <0 0 0>, <0 0 0>, <0 0 0>]
    trailBone {boneName=cape_mid_06_r}
    trailRelaxationSpeed {editorCurveData {keys [<at 0 15 RCIM_Linear>,<at 0.5289 13.1838 RCIM_Linear>,<at 1 6 RCIM_Linear>]}}
   }
   n68 = Trail controller - Bone: cape_06_l(ComponentPose=@n67, Alpha=@n65) {
    bUsePlanarLimit=true
    chainLength=6
    debugLifeTime=0.1
    maxDeltaTime=0.0333
    alphaScaleBiasClamp {bInterpResult=true}
    rotationLimits [
     {limitMax=<180 180 180>}
     {limitMax=<180 180 180>}
     {limitMax=<180 90 180>, limitMin=<-180 -90 -180>}
     {limitMax=<180 180 180>}
     {limitMax=<180 180 180>}
     {limitMax=<180 180 180>}
    ]
    rotationOffsets [<0 0 0>, <0 0 0>, <0 0 0>, <0 0 0>, <0 0 0>, <0 0 0>]
    trailBone {boneName=cape_06_l}
    trailRelaxationSpeed {editorCurveData {keys [<at 0 15 RCIM_Linear>,<at 0.5289 13.1838 RCIM_Linear>,<at 1 6 RCIM_Linear>]}}
   }
   n69 = Trail controller - Bone: cape_06_r(ComponentPose=@n68, Alpha=@n65) {
    bUsePlanarLimit=true
    chainLength=6
    debugLifeTime=0.1
    maxDeltaTime=0.0333
    alphaScaleBiasClamp {bInterpResult=true}
    rotationLimits [
     {limitMax=<180 180 180>}
     {limitMax=<180 180 180>}
     {limitMax=<180 90 180>, limitMin=<-180 -90 -180>}
     {limitMax=<180 180 180>}
     {limitMax=<180 180 180>}
     {limitMax=<180 180 180>}
    ]
    rotationOffsets [<0 0 0>, <0 0 0>, <0 0 0>, <0 0 0>, <0 0 0>, <0 0 0>]
    trailBone {boneName=cape_06_r}
    trailRelaxationSpeed {editorCurveData {keys [<at 0 15 RCIM_Linear>,<at 0.5289 13.1838 RCIM_Linear>,<at 1 6 RCIM_Linear>]}}
   }
   n70 = Component To Local(ComponentPose=@n69)
   n71 = Get BwdMovement()
   n72 = Trail controller - Bone: cape_mid_06_l(ComponentPose=Local To Component(LocalPose=Use cached pose 'RBANCache'()), Alpha=@n71) {
    bUsePlanarLimit=true
    chainLength=4
    debugLifeTime=0.1
    maxDeltaTime=0.0333
    alphaScaleBiasClamp {bInterpResult=true}
    rotationLimits [{limitMax=<180 180 180>}, {limitMax=<180 180 180>}, {limitMax=<180 90 180>, limitMin=<-180 -90 -180>}, {limitMax=<180 180 180>}]
    rotationOffsets [<0 0 0>, <0 0 0>, <0 0 0>, <0 0 0>]
    trailBone {boneName=cape_mid_06_l}
    trailRelaxationSpeed {editorCurveData {keys [<at 0 15 RCIM_Linear>,<at 0.5289 13.1838 RCIM_Linear>,<at 1 6 RCIM_Linear>]}}
   }
   n73 = Trail controller - Bone: cape_mid_06_r(ComponentPose=@n72, Alpha=@n71) {
    bUsePlanarLimit=true
    chainLength=4
    debugLifeTime=0.1
    maxDeltaTime=0.0333
    alphaScaleBiasClamp {bInterpResult=true}
    rotationLimits [{limitMax=<180 180 180>}, {limitMax=<180 180 180>}, {limitMax=<180 180 180>}, {limitMax=<180 180 180>}]
    rotationOffsets [<0 0 0>, <0 0 0>, <0 0 0>, <0 0 0>]
    trailBone {boneName=cape_mid_06_r}
    trailRelaxationSpeed {editorCurveData {keys [<at 0 15 RCIM_Linear>,<at 0.5289 13.1838 RCIM_Linear>,<at 1 6 RCIM_Linear>]}}
   }
   n74 = Trail controller - Bone: cape_06_l(ComponentPose=@n73, Alpha=@n71) {
    bUsePlanarLimit=true
    chainLength=4
    debugLifeTime=0.1
    maxDeltaTime=0.0333
    alphaScaleBiasClamp {bInterpResult=true}
    rotationLimits [{limitMax=<180 180 180>}, {limitMax=<180 180 180>}, {limitMax=<180 90 180>, limitMin=<-180 -90 -180>}, {limitMax=<180 180 180>}]
    rotationOffsets [<0 0 0>, <0 0 0>, <0 0 0>, <0 0 0>]
    trailBone {boneName=cape_06_l}
    trailRelaxationSpeed {editorCurveData {keys [<at 0 15 RCIM_Linear>,<at 0.5289 13.1838 RCIM_Linear>,<at 1 6 RCIM_Linear>]}}
   }
   n75 = Trail controller - Bone: cape_06_r(ComponentPose=@n74, Alpha=@n71) {
    bUsePlanarLimit=true
    chainLength=4
    debugLifeTime=0.1
    maxDeltaTime=0.0333
    alphaScaleBiasClamp {bInterpResult=true}
    rotationLimits [{limitMax=<180 180 180>}, {limitMax=<180 180 180>}, {limitMax=<180 90 180>, limitMin=<-180 -90 -180>}, {limitMax=<180 180 180>}]
    rotationOffsets [<0 0 0>, <0 0 0>, <0 0 0>, <0 0 0>]
    trailBone {boneName=cape_06_r}
    trailRelaxationSpeed {editorCurveData {keys [<at 0 15 RCIM_Linear>,<at 0.5289 13.1838 RCIM_Linear>,<at 1 6 RCIM_Linear>]}}
   }
   n76 = Component To Local(ComponentPose=@n75)
   n77 = Get Is Wearing Backbling()
   Save cached pose 'PreSleeveSim'(Pose=Blend Poses by bool(
     BlendPose_0=Use cached pose 'standingPoseCache'()
     BlendPose_1=Blend Poses by bool(BlendPose_0=@n76, BlendPose_1=@n70, bActiveValue=@n77) {blendType=Linear}
     bActiveValue=Get bIsDBNOCarried()
    ) {blendType=Linear}) {cacheName=PreSleeveSim}
   n78 = RigidBody!(
    ComponentPose=Local To Component(LocalPose=Use cached pose 'standingPoseCache'())
    Alpha
    OverrideWorldGravity=Get Gravity Override Cape()
    ExternalForce
    SimSpaceSettings=(WorldAlpha=1,VelocityScaleZ=1,DampingAlpha=1,MaxLinearVelocity=3000,MaxAngularVelocity=5,MaxLinearAcceleration=3000,MaxAngularAcceleration=3,ExternalLinearDragV=(X=0,Y=0,Z=0),ExternalLinearVelocity=(X=0,Y=0,Z=0),ExternalAngularVelocity=(X=0,Y=0,Z=0))
   ) -> (Pose) {
    bClampLinearTranslationLimitToRefPose=true
    bOverrideWorldGravity=true
    overridePhysicsAsset=&F_MED_BinGrass_Cape_Physics
    simulationSpace=BaseBoneSpace
    alphaBoolBlend {blendOption=Linear}
    baseBoneRef {boneName=gravity}
    overrideWorldGravity {z=-980}
    simSpaceSettings {maxAngularAcceleration=3, maxAngularVelocity=5, maxLinearAcceleration=3000, maxLinearVelocity=3000, worldAlpha=1}
   }
   n79 = Component To Local(ComponentPose=@n78)
   n80 = RigidBody(
    ComponentPose=Local To Component(LocalPose=Use cached pose 'standingPoseCache'())
    Alpha=Get in vehicle rigid alpha()
    OverrideWorldGravity=Get Gravity Override Cape()
    SimSpaceSettings=(WorldAlpha=0,VelocityScaleZ=0,DampingAlpha=1,MaxLinearVelocity=150,MaxAngularVelocity=1,MaxLinearAcceleration=300,MaxAngularAcceleration=50,ExternalLinearDragV=(X=0.2,Y=0.2,Z=0),ExternalLinearVelocity=(X=0,Y=0,Z=0),ExternalAngularVelocity=(X=0,Y=0,Z=0))
   ) {
    bClampLinearTranslationLimitToRefPose=true
    bOverrideWorldGravity=true
    overridePhysicsAsset=&F_MED_BinGrass_Cape_Physics
    simulationSpace=BaseBoneSpace
    alphaBoolBlend {blendOption=Linear}
    baseBoneRef {boneName=gravity}
    overrideWorldGravity {z=-980}
    simSpaceSettings {
     maxAngularAcceleration=50
     maxAngularVelocity=1
     maxLinearAcceleration=300
     maxLinearVelocity=150
     velocityScaleZ=0
     externalLinearDragV=<0.2 0.2>
    }
   }
   n81 = Component To Local(ComponentPose=@n80)
   n82 = Get bIsTacticalSprinting()
   n83 = RigidBody(
    ComponentPose=Local To Component(LocalPose=Use cached pose 'standingPoseCache'())
    Alpha=Get in vehicle rigid alpha()
    OverrideWorldGravity=Get Gravity Override Cape()
    SimSpaceSettings=(WorldAlpha=1,VelocityScaleZ=1,DampingAlpha=1,MaxLinearVelocity=150,MaxAngularVelocity=1,MaxLinearAcceleration=300,MaxAngularAcceleration=50,ExternalLinearDragV=(X=0.2,Y=0.2,Z=0),ExternalLinearVelocity=(X=0,Y=0,Z=0),ExternalAngularVelocity=(X=0,Y=0,Z=0))
   ) {
    bClampLinearTranslationLimitToRefPose=true
    bOverrideWorldGravity=true
    overridePhysicsAsset=&F_MED_BinGrassBackBling_Cape_Physics2
    simulationSpace=BaseBoneSpace
    alphaBoolBlend {blendOption=Linear}
    baseBoneRef {boneName=gravity}
    overrideWorldGravity {z=-980}
    simSpaceSettings {
     maxAngularAcceleration=50
     maxAngularVelocity=1
     maxLinearAcceleration=300
     maxLinearVelocity=150
     worldAlpha=1
     externalLinearDragV=<0.2 0.2>
    }
   }
   Save cached pose 'posBackbing'(Pose=Blend Poses by bool(
     BlendPose_0=Component To Local(ComponentPose=@n83)
     BlendPose_1=Blend Poses by bool(BlendPose_0=@n81, BlendPose_1=@n79, BlendTime_1=0.7, bActiveValue=@n82) {blendType=Linear}
     BlendTime_1=0.7
     bActiveValue=Get Is Wearing Backbling()
    ) {blendType=Linear}) {cacheName=posBackbing}
  ]}
 EventGraph {nodes [
   n1 = Event BlueprintUpdateAnimation()
   > Set DeltaTimeLocal(DeltaTimeLocal:double=@n1.DeltaTimeX) -> (Output_Get:double)
   > n2 = Sequence!(execute) -> (then_0, then_1, then_2:Exec, then_3:Exec, then_4:Exec, then_5:Exec, then_6:Exec, then_7:Exec, then_8:Exec, then_9:Exec)
   > Set Body Mesh(Body Mesh:Skeletal Mesh Component Object Reference=GetPartSkeletalMeshComponent(self=Self-Reference())) -> (Output_Get:Skeletal Mesh Component Object Reference)
   n3 = Gravity Override State Function DEPRECATED(
    execute=@n2.then_1
    CharacterPartAnimInstance=Self-Reference()
    GameTimeInSeconds=GetGameTimeInSeconds()
    GravityOverrideParams=Get Gravity Override Params()
    RBAN Params=Get RBAN Params()
   )
   > Set Gravity Override Cape(Gravity Override Cape:Vector=@n3) -> (Output_Get:Vector)
   > Set RBAN Settings Out(RBAN Settings Out:RBANSetting Struct Structure=@n3.RBAN SettingsOut) -> (Output_Get:RBANSetting Struct Structure)
   Cloth Controller Function DEPRECATED!(
    execute=@n2.then_2
    CharacterPartAnimInstance:Custom Character Part Anim Instance Object Reference=Self-Reference()
    ClothingAssetName:String=F_M_BinGrass_Clothing_0
    Cloth Paramaters:Cloth Params Struct Structure=Get Cloth Paramaters_0() -> (Cloth Paramaters_0:Cloth Params Struct Structure)
    Cloth Core Settings:Cloth Core Setting Struct Structure=Get Cloth Core Settings_0() -> (Cloth Core Settings_0:Cloth Core Setting Struct Structure)
    ClothUpAxis:Vector="1,0,0"
    OnGround_Moving_SpeedLimit:double=2
    OnGround_Moving_AngularVel_Max:double=0.05
    OnGround_Moving_FictitiousVel_Max:double=0.6
    LinVelScale_FwdBwd_Clamp:double=0
    LinVelScale_UpDwn_Clamp:double=0.7
    DeltaTime:double=Get DeltaTimeLocal() -> (DeltaTimeLocal:double)
    GameTimeInSeconds:double=GetGameTimeInSeconds()
    WindGustsParams:Wind Gust Params Struct Structure=Get Wind Gust Params() -> (Wind Gust Params:Wind Gust Params Struct Structure)
    WindGustLocalSettings:Wind Gust Local Setting Struct Structure=Get Gust Local Settings() -> (Gust Local Settings:Wind Gust Local Setting Struct Structure)
   ) -> (
    then
    CurrentStateOut:String
    Cloth Settings Out:Cloth Setting Struct Structure
    Cloth Core Settings Out:Cloth Core Setting Struct Structure
    WindGustSettings Out:Wind Gust Setting Struct Structure
    WindGustLocalSettings Out:Wind Gust Local Setting Struct Structure
    GustStrength - WindVelocityOut:Vector
    LinearVelVector-DELETE LATER:Vector
   )
   Sequence!(execute=@n2.then_3) -> (then_0, then_1, then_2, then_3)
   > n4 = Vehicle Dynamics Scaler Function!(execute, CharacterPartAnimInstance:Custom Character Part Anim Instance Object Reference=Self-Reference(), Debug:bool=false) -> (then, InVehicleHeadScaleAlpha:double, InVehicleHeadSimAlpha:double, InVehicleSkirtScaleAlpha:double, InVehicleSkirtSimAlpha:double)
   > Set in vehicle scale alpha(in vehicle scale alpha:double=@n4.InVehicleSkirtScaleAlpha) -> (Output_Get:double)
   > Set in vehicle scale alpha(in vehicle scale alpha=@n4.InVehicleSkirtSimAlpha)
   n5 = Get bIsPlayingEmote()
   n6 = Get bIsInFrontEnd()
   Set FrontEnd or Emote(execute=@n2.then_4, FrontEnd or Emote:bool=OR Boolean(A=@n6, B=@n5)) -> (Output_Get:bool)
   > Set FrontEnd or NotEmote(FrontEnd or NotEmote:bool=AND Boolean!(A=@n6, B=NOT Boolean(A=@n5), C:bool=Get Is Wearing Backbling()) -> (ReturnValue)) -> (Output_Get:bool)
   n7 = NOT Boolean(A=Get bIsInFrontEnd())
   n8 = Get bIsPlayingEmote()
   > Set Crouching or Emote or Siding(Crouching or Emote or Siding:bool=OR Boolean!(A=AND Boolean!(A=@n8, B=@n7) -> (ReturnValue), B=Get bIsCrouching(), C:bool=Get bIsSliding()) -> (ReturnValue)) -> (Output_Get:bool)
   n9 = GetAnimInstance(self=GetOwnerSkeletalMeshComponent())
   n10 = OR Boolean!(
    A=Montage_IsPlaying(self=@n9, Montage=/BRCosmetics/Animation/Game/MainPlayer/Emotes/BinGrass/Frontend/Emote_BinGrass_FE_CMF_M)
    B=Montage_IsPlaying(self=@n9, Montage=/BRCosmetics/Animation/Game/MainPlayer/Emotes/BinGrass/Emote_BinGrass_CMF_M)
   ) -> (ReturnValue)
   n11 = Branch(execute=@n2.then_5, Condition=@n10)
   n12 = SelectFloat(A=1, B=0, bPickA=@n10)
   > Set Is Playing Custom Emote(Is Playing Custom Emote:double=@n12) -> (Output_Get:double)
   > Set emote_bingrass(emote_bingrass:bool=Get bRigidBodyEnabled()) -> (Output_Get:bool)
   Set Is Playing Custom Emote(execute=@n11.else, Is Playing Custom Emote=@n12)
   > Set emote_bingrass(emote_bingrass=false)
   n13 = Branch(
    execute=@n2.then_7
    Condition=OR Boolean!(A=Get emote_bingrass(), B=Get Is Wearing Backbling(), C:bool=Get bIsDBNOCarried(), D:bool=Get bIsDBNOCarrying() -> (bIsDBNOCarrying:bool)) -> (ReturnValue)
   )
   n14 = GetPartSkeletalMeshComponent(self=Self-Reference())
   Set ClothBlendWeight!(execute=[@n15.then, @n13.then], ClothBlendWeight:float=0, self:Skeletal Mesh Component Object Reference=@n14) -> (then, Output_Get:float)
   > Set bDisableClothSimulation!(execute, bDisableClothSimulation:bool=true, self:Skeletal Mesh Component Object Reference=@n14) -> (then, Output_Get:bool)
   n15 = Branch(
    execute=@n13.else
    Condition=Get bIsInFirstPerson!(self:Fort Player Anim Instance Object Reference=GetOwnerFortPlayerAnimInstance!(self:Custom Character Part Anim Instance Object Reference) -> (ReturnValue:Fort Player Anim Instance Object Reference)) -> (bIsInFirstPerson:bool)
   )
   Set ClothBlendWeight(execute=@n15.else, ClothBlendWeight=1, self=@n14)
   > Set bDisableClothSimulation(bDisableClothSimulation=false, self=@n14)
   n16 = Branch(execute=@n2.then_8, Condition=Get bRigidBodyEnabled())
   > Set TrailControllerRot(TrailControllerRot:Rotator="0, 0, 0") -> (Output_Get:Rotator)
   > Set BwdMovement(BwdMovement:double=0) -> (Output_Get:double)
   > Set reposeCapeTrail(reposeCapeTrail:bool=false) -> (Output_Get:bool)
   > Set CRAlpha(CRAlpha:double=1) -> (Output_Get:double)
   n17 = Sequence!(execute=@n16.else) -> (then_0, then_1)
   > Set reposeCapeTrail(reposeCapeTrail=true)
   > Set CRAlpha(CRAlpha=0)
   > n18 = PlayerMovement(AnimInstance=Self-Reference(), XDotProductCoefficient=1, YDotProductCoefficient=1, ZDotProductCoefficient=0)
   n19 = MapRangeClamped(Value=@n18.YMovement, InRangeA=0, InRangeB=1)
   n20 = "Absolute (Float)"(A=@n18)
   > Set BwdMovement(BwdMovement=MapRangeClamped(Value=float + float(A=@n20, B=@n19) {is=title K2Node_CommutativeAssociativeBinaryOperator}, InRangeA=0, InRangeB=1.4))
   > n21 = Branch(Condition=Get bIsSwimSprinting() -> (bIsSwimSprinting:bool))
   > Set TrailControllerRot(TrailControllerRot="0,-15, 0")
   n22 = Branch(execute=@n21.else, Condition=Get bIsSkyDiving())
   > n23 = Branch(Condition=Get bIsParachuteOpened())
   > Set TrailControllerRot(TrailControllerRot="0,-15,0")
   > Set BwdMovement(BwdMovement=0)
   n24 = Branch(execute=@n23.else, Condition="float > float"(A=Get PawnVelocityZ(), B=50) {is=title K2Node_CallFunction})
   > Set BwdMovement(BwdMovement=float * float(A=Get BwdMovement(), B=0))
   > Set TrailControllerRot(TrailControllerRot="0,-25, 0")
   n25 = Branch(execute=@n24.else, Condition=Get bIsSkydiveDiveMode())
   > Set TrailControllerRot(TrailControllerRot="0,-15,0")
   Set TrailControllerRot(execute=@n25.else, TrailControllerRot="0,-25,0")
   Set TrailControllerRot(execute=@n22.else, TrailControllerRot="0,-12.5, 0")
   > Set BwdMovement(BwdMovement=float * float(A=Get BwdMovement(), B=1))
   Set skydive(execute=@n17.then_1, skydive:bool=OR Boolean!(A=Get bIsSkyDiving(), B=Get bIsSkydiveFloating(), C=Get bIsSkydiveDiveMode()) -> (ReturnValue)) -> (Output_Get:bool)
   > n26 = Branch(Condition=Get skydive() -> (skydive:bool))
   > Set clavicle_float_r(clavicle_float_r:double=0) -> (Output_Get:double)
   > Set clavicle_float_l(clavicle_float_l:double=0) -> (Output_Get:double)
   n27 = GetDeltaTransformFromRefPose(self=GetPartSkeletalMeshComponent(), BoneName=upperarm_l, BaseName=spine_05)
   n28 = MapRangeClamped(Value=BreakRotator(InRot=BreakTransform(InTransform=@n27).Rotation).Yaw, InRangeA=10, InRangeB=-30)
   > Set clavicle_glider_float_l(clavicle_glider_float_l:double=@n28) -> (Output_Get:double)
   n29 = BreakTransform(InTransform=GetDeltaTransformFromRefPose(self=GetPartSkeletalMeshComponent(), BoneName=upperarm_r, BaseName=spine_05))
   Set clavicle_float_r(execute=@n26.else, clavicle_float_r=MapRangeClamped(Value=BreakRotator(InRot=@n29.Rotation).Yaw, InRangeA=0, InRangeB=-40))
   > Set clavicle_float_l(clavicle_float_l=@n28)
   > Set clavicle_glider_float_l(clavicle_glider_float_l=0)
   n30 = Get bIsPlayingEmote()
   n31 = Get bRigidBodyEnabled()
   n32 = Branch(
    execute=@n2.then_6
    Condition=AND Boolean!(A=NOT Boolean(A=@n30), B=NOT Boolean(A=@n31), C:bool=Get NotClothEnabled() -> (NotClothEnabled:bool), D:bool=Get bIsInFrontEnd()) -> (ReturnValue)
   )
   > Set Is_Emoting_FE(Is_Emoting_FE:bool=true) -> (Output_Get:bool)
   Set Is_Emoting_FE(execute=@n32.else, Is_Emoting_FE=false)
   Event BlueprintInitializeAnimation()
   n33 = GetOwnerSkeletalMeshComponent()
   > Is Valid(InputObject=@n33)
   > Set Skel_Mesh(Skel_Mesh=@n33)
   > n34 = Cast To FortPlayerPawn(Object=GetOwningActor())
   > n35 = Set Pawn(Pawn:Fort Player Pawn Object Reference=@n34) -> (Output_Get:Fort Player Pawn Object Reference)
   > Set Anim_Inst(Anim_Inst:Anim Instance Object Reference=GetAnimInstance(self=Get Mesh(self=@n35))) -> (Output_Get:Anim Instance Object Reference)
   > n36 = Cast To FortPlayerPawn(Object=TryGetPawnOwner())
   > Set As Fort Player Pawn(As Fort Player Pawn:Fort Player Pawn Object Reference=@n36) -> (Output_Get:Fort Player Pawn Object Reference)
   Event OnCharacterCustomizationLoaded()
   > n37 = Sequence()
   n38 = CheckForBackbling()
   > n39 = Branch(Condition=OR Boolean!(A=@n38, B=@n38.bIsBackblingLong) -> (ReturnValue))
   > Set Is Wearing Backbling(Is Wearing Backbling=true)
   Set Is Wearing Backbling(execute=@n39.else, Is Wearing Backbling=false)
   > Set Is Wearing Backbling(Is Wearing Backbling=Get bIsDBNOCarrying())
   Set Not Wearing Backbling(execute=@n37.then_1, Not Wearing Backbling:bool=NOT Boolean(A=Get Is Wearing Backbling())) -> (Output_Get:bool)
   K2Node_Event_4 = CustomEvent() {is=title K2Node_CustomEvent}
   > Set TrailControllerRot(TrailControllerRot="0,-65,0")
   > Delay(Duration=0.3)
   > ResetDynamics(InTeleportType=TeleportPhysics)
   > ResetDynamics(InTeleportType=ResetPhysics)
   K2Node_Event_6 = CustomEvent_1() {is=title K2Node_CustomEvent}
   > Set TrailControllerRot(TrailControllerRot="0,-10,0")
   Branch(Condition=Get emote_bingrass())
   > SetClothingInteractorIfNotSet()
   > n40 = Gravity Override State Function DEPRECATED(
    CharacterPartAnimInstance=Self-Reference()
    GameTimeInSeconds=GetGameTimeInSeconds()
    GravityOverrideParams=Get Gravity Override Params_0() -> (Gravity Override Params_0:Gravity Override Params Struct Structure)
   )
   n41 = Get ChaosClothingInteractor() -> (ChaosClothingInteractor:Chaos Clothing Interactor Object Reference)
   > SetGravity!(
    execute
    self:Chaos Clothing Interactor Object Reference=@n41
    GravityScale:float=0
    bIsGravityOverridden:bool=true
    GravityOverride:Vector=@n40
   ) -> (then)
   > SetVelocityScale!(
    execute
    self:Chaos Clothing Interactor Object Reference=@n41
    LinearVelocityScale:Vector="0.3,0.3,0.1"
    AngularVelocityScale:float=0.3
    FictitiousAngularScale:float=0.1
   ) -> (then)
   > SetAerodynamics!(
    execute
    self:Chaos Clothing Interactor Object Reference=@n41
    DragCoefficient:float=0.7
    LiftCoefficient:float=1
    WindVelocity:Vector="0,0,270"
   ) -> (then)
  ]}
 SetClothingInteractorIfNotSet {nodes [
   SetClothingInteractorIfNotSet() {is=title K2Node_FunctionEntry}
   n1 = IsValid(Object=Get ChaosClothingInteractor())
   > Branch(Condition=AND Boolean!(A=Get bClothEnabled(), B=NOT Boolean(A=@n1)) -> (ReturnValue))
   n2 = MakeLiteralString(Value=F_M_BinGrass_Clothing_0)
   n3 = GetOwningComponent()
   > n4 = Cast To ChaosClothingInteractor!(
    execute
    Object=GetClothingInteractor!(
     self:Clothing Simulation Interactor Object Reference=GetClothingSimulationInteractor!(self:Skeletal Mesh Component Object Reference=@n3, ClothingAssetName:Name=None) -> (ReturnValue:Clothing Simulation Interactor Object Reference)
     ClothingAssetName:Name=String To Name(InString=@n2)
    ) -> (ReturnValue:Clothing Interactor Object Reference)
   ) -> (then, CastFailed, AsChaos Clothing Interactor:Chaos Clothing Interactor Object Reference) {is=title K2Node_DynamicCast}
   > Set ChaosClothingInteractor(ChaosClothingInteractor:Chaos Clothing Interactor Object Reference=@n4) -> (Output_Get:Chaos Clothing Interactor Object Reference)
  ]}
]
variables [
 Anim_Inst {type=AnimInstance, default=null}
 arms {type=boolean}
 As Fort Player Pawn {type=FortPlayerPawn}
 As Fort Player Pawn_0 {type=FortPlayerPawn}
 Body Mesh {type=SkeletalMeshComponent}
 BwdMovement {type=number}
 ChaosClothingInteractor {type=ChaosClothingInteractor, default=null}
 clavicle_float_l {type=number}
 clavicle_float_r {type=number}
 clavicle_glider_float_l {type=number}
 Cloth Core Settings_0 {
  type=ClothCoreSettingStruct
  default {
   collisionThickness=1
   dampingCoefficient=0.2
   frictionCoefficient=0.3
   localDampingCoefficient=0
   numberIterations=1
   numberSubsteps=2
   selfCollisionThickness=0
   animDriveDamping=<0 1>
   animDriveStiffness=<0 1>
   areaStiffness=<1 1>
   bendingStiffness=<1 1>
   edgeStiffness=<1 1>
   tetherStiffness=<1 1>
  }
 }
 Cloth Paramaters_0 {
  type=ClothParamsStruct
  default {
   dBNO {
    angularVelocityScale=0.6
    fictitiousAngularScale=0.1
    jointName=pelvis
    dragCoefficient=<0.2 1>
    gravityOverride=<-980 400 0>
    liftCoefficient=<0 0.5>
    linearVelocityScale=<0.3 0.3 0.1>
   }
   dBNO_Carried {
    angularVelocityScale=0
    fictitiousAngularScale=0
    jointName=pelvis
    dragCoefficient=<0 0>
    gravityOverride=<-980 400 0>
    liftCoefficient=<0 0>
    linearVelocityScale=<0 0 0>
   }
   dBNO_Impulse {
    angularVelocityScale=0
    fictitiousAngularScale=0
    jointName=pelvis
    dragCoefficient=<0 0.2>
    gravityOverride=<-3000 -2000 0>
    liftCoefficient=<0 0.1>
    linearVelocityScale=<0.3 0.1 0.3>
   }
   dBNO_RollBwd {
    angularVelocityScale=0
    fictitiousAngularScale=0
    jointName=pelvis
    dragCoefficient=<0 0.2>
    gravityOverride=<-2000 0 0>
    liftCoefficient=<0 0.1>
    linearVelocityScale=<0.3 0.1 0.3>
   }
   dBNO_RollFwd {
    angularVelocityScale=0
    fictitiousAngularScale=0
    jointName=pelvis
    dragCoefficient=<0 0.2>
    gravityOverride=<-980 0 0>
    liftCoefficient=<0 0.1>
    linearVelocityScale=<0.3 0.1 0.3>
   }
   dBNO_Tumbling {
    angularVelocityScale=0
    fictitiousAngularScale=0
    jointName=pelvis
    dragCoefficient=<0 0.2>
    gravityOverride=<-2500 -300 0>
    liftCoefficient=<0 0.1>
    linearVelocityScale=<0.3 0.1 0.3>
   }
   emoteOrMelee {
    angularVelocityScale=0.3
    fictitiousAngularScale=1
    jointName=root
    dragCoefficient=<1 1>
    gravityOverride=<0 -400 -980>
    liftCoefficient=<0.1 0.15>
    linearVelocityScale=<0.3 0.3 0.3>
   }
   front_End {
    angularVelocityScale=0.1
    fictitiousAngularScale=0.1
    jointName=pelvis
    dragCoefficient=<0.1 0.5>
    gravityOverride=<-980 -50 50>
    liftCoefficient=<0.1 0.1>
    linearVelocityScale=<0.2 0.2 0.3>
   }
   onGround_Crouch_Moving {
    angularVelocityScale=0.1
    fictitiousAngularScale=0.5
    jointName=root
    dragCoefficient=<0.5 0.5>
    gravityOverride=<0 -500 -700>
    liftCoefficient=<0.1 0.2>
    linearVelocityScale=<0.4 0.4 0.3>
   }
   onGround_Falling {
    angularVelocityScale=0.6
    fictitiousAngularScale=0
    jointName=root
    dragCoefficient=<0 0.2>
    gravityOverride=<0 -500 -980>
    liftCoefficient=<0 0.1>
    linearVelocityScale=<2 1 0.3>
   }
   onGround_Moving {
    angularVelocityScale=0.05
    fictitiousAngularScale=0
    jointName=root
    dragCoefficient=<0.1 0.5>
    gravityOverride=<0 -400 -500>
    liftCoefficient=<0.1 0.5>
    linearVelocityScale=<0.2 0.1 0.1>
   }
   onGround_Standing {
    angularVelocityScale=0.3
    fictitiousAngularScale=0.1
    jointName=root
    dragCoefficient=<0.1 0.2>
    gravityOverride=<0 -50 -980>
    liftCoefficient=<0.1 0.2>
    linearVelocityScale=<0.15 0.3 0.3>
   }
   rail_Grinding {
    angularVelocityScale=0.2
    fictitiousAngularScale=0
    jointName=root
    dragCoefficient=<0.1 0.2>
    gravityOverride=<50 -500 -400>
    liftCoefficient=<0 0.5>
    linearVelocityScale=<0.1 0.1 0.1>
   }
   riding_Moving {
    angularVelocityScale=0.05
    fictitiousAngularScale=0.1
    jointName=pelvis
    dragCoefficient=<0.1 0.1>
    gravityOverride=<-980 0 0>
    liftCoefficient=<0.1 0.1>
    linearVelocityScale=<0.4 0.6 0.3>
   }
   riding_Moving_Targeting {
    angularVelocityScale=0.05
    fictitiousAngularScale=0.7
    jointName=root
    dragCoefficient=<0.5 0.5>
    gravityOverride=<-980 0 0>
    liftCoefficient=<0.5 0.5>
    linearVelocityScale=<0.2 0.6 0.2>
   }
   riding_Standing {
    angularVelocityScale=0.05
    fictitiousAngularScale=1
    jointName=pelvis
    dragCoefficient=<0.5 0.5>
    gravityOverride=<-980 -450 0>
    liftCoefficient=<0.5 0.5>
    linearVelocityScale=<0.6 0.6 0.3>
   }
   skydive_DivingDown {
    angularVelocityScale=0.1
    fictitiousAngularScale=0
    jointName=pelvis
    dragCoefficient=<0.1 1>
    gravityOverride=<0 -200 -150>
    liftCoefficient=<0.2 0.5>
    linearVelocityScale=<0.1 0.1 0.1>
   }
   skydive_FloatingDown {
    angularVelocityScale=0.1
    fictitiousAngularScale=0
    jointName=root
    dragCoefficient=<0 1>
    gravityOverride=<0 -600 -150>
    liftCoefficient=<0.1 0.2>
    linearVelocityScale=<0.2 0.2 0.1>
   }
   skydive_GoingUp {
    angularVelocityScale=0.1
    fictitiousAngularScale=0
    jointName=root
    dragCoefficient=<0.1 1>
    gravityOverride=<0 -150 150>
    liftCoefficient=<0.2 0.1>
    linearVelocityScale=<0.01 0.01 0.01>
   }
   skydive_Parachuting {
    angularVelocityScale=0.1
    fictitiousAngularScale=0
    jointName=root
    dragCoefficient=<0.1 1>
    gravityOverride=<0 -200 -980>
    liftCoefficient=<0.2 1>
    linearVelocityScale=<0.1 0.1 0.1>
   }
   skydive_Surfing {
    angularVelocityScale=0.1
    fictitiousAngularScale=0
    jointName=root
    dragCoefficient=<0.1 1>
    gravityOverride=<200 -50 -800>
    liftCoefficient=<0.2 1>
    linearVelocityScale=<0.1 0.1 0.1>
   }
   swimming {
    angularVelocityScale=0.4
    fictitiousAngularScale=0
    jointName=root
    dragCoefficient=<0 0.5>
    gravityOverride=<0 0 -400>
    liftCoefficient=<0 0.1>
    linearVelocityScale=<0.4 0.2 0.2>
   }
   tactical_Sprint {
    angularVelocityScale=0.1
    fictitiousAngularScale=0
    jointName=root
    dragCoefficient=<0.1 0.1>
    gravityOverride=<0 -200 -150>
    liftCoefficient=<0.1 0.1>
    linearVelocityScale=<0.35 0.2 0.2>
   }
   ziplining {
    angularVelocityScale=0.6
    fictitiousAngularScale=0.2
    jointName=root
    dragCoefficient=<0 1.5>
    gravityOverride=<0 -200 -400>
    liftCoefficient=<0 0.3>
    linearVelocityScale=<0 0 0>
   }
  }
 }
 cloth_on/off {type=boolean, default=true}
 CRAlpha {type=number, default=1}
 Crouching or Emote or Siding {type=boolean}
 DeltaTimeLocal {type=number, default=0.0266}
 emote_bingrass {type=boolean}
 FrontEnd or Emote {type=boolean}
 FrontEnd or NotEmote {type=boolean}
 Gravity Override Cape {type=Vector}
 Gravity Override Params {
  type=GravityOverrideParamsStruct
  default {
   emoteOrMelee {gravityOverride=<0 -150 -800>}
   onGround_Moving {gravityOverride=<0 -300 -800>}
   onGround_Standing {gravityOverride=<0 -150 -980>}
   skydive_GoingUp {gravityOverride=<0 -300 600>}
   tactical_Sprint {gravityOverride=<0 0 80>}
  }
 }
 Gravity Override Params_0 {
  type=GravityOverrideParamsStruct
  default {
   emoteOrMelee {gravityOverride=<0 -600 -980>, windAmplitude=<0 -500 150>, windFrequency=<15 6 2>, windOffset=<0 -800 -150>}
   front_End {gravityOverride=<0 -600 -980>, windAmplitude=<0 -500 -100>, windFrequency=<15 6 2>, windOffset=<0 -800 -150>}
   onGround_Moving {gravityOverride=<0 -600 -980>, windAmplitude=<0 -500 -150>, windFrequency=<15 6 2>, windOffset=<0 -800 -400>}
   onGround_Standing {gravityOverride=<0 -150 -980>}
   skydive_GoingUp {gravityOverride=<0 -300 600>}
   tactical_Sprint {gravityOverride=<0 0 80>}
  }
 }
 Gust Local Settings {type=WindGustLocalSettingStruct, default {elapsedTime=0, finalGustStrengthLocal=0, gustDuration=0, gustStrength=0, timeBetweenGusts=0}}
 in vehicle rigid alpha {type=number, default=1}
 in vehicle scale alpha {type=number, default=1}
 Is Playing Custom Emote {type=number}
 Is Wearing Backbling {type=boolean}
 Is_Emoting_FE {type=boolean}
 ISDBNO {type=boolean, default=true}
 IsNotFrontend {type=boolean, default=true}
 Not Wearing Backbling {type=boolean, default=true}
 NotClothEnabled {type=boolean, default=true}
 Pawn {type=FortPlayerPawn}
 RBAN Params {
  type=RBANParamsStruct
  default {
   dBNO {simSpaceSettings {
     maxAngularAcceleration=3
     maxAngularVelocity=5
     maxLinearAcceleration=500
     maxLinearVelocity=1500
     velocityScaleZ=1
     externalLinearDragV=<1 1 0>
    }}
   dBNO_Carried {simSpaceSettings {maxAngularAcceleration=1, maxAngularVelocity=1, maxLinearAcceleration=10, maxLinearVelocity=0, velocityScaleZ=1, externalLinearDragV=<1 1 0>}}
   emoteOrMelee {
    componentLinearAccScale=<1 1 1>
    componentLinearVelScale=<1.1 1.1 1.1>
    simSpaceSettings {
     maxAngularAcceleration=3
     maxAngularVelocity=5
     maxLinearAcceleration=500
     maxLinearVelocity=1500
     velocityScaleZ=1
     externalLinearDragV=<1 1 0>
    }
   }
   front_End {simSpaceSettings {
     maxAngularAcceleration=1
     maxAngularVelocity=3.5
     maxLinearAcceleration=10
     maxLinearVelocity=500
     velocityScaleZ=1
     externalLinearDragV=<1 1 0>
    }}
   onGround_Crouch_Moving {simSpaceSettings {
     maxAngularAcceleration=3
     maxAngularVelocity=5
     maxLinearAcceleration=500
     maxLinearVelocity=1500
     velocityScaleZ=1
     externalLinearDragV=<1 1 0>
    }}
   onGround_Falling {simSpaceSettings {
     maxAngularAcceleration=3
     maxAngularVelocity=5
     maxLinearAcceleration=500
     maxLinearVelocity=1500
     velocityScaleZ=1
     externalLinearDragV=<1 1 0>
    }}
   onGround_Moving {
    componentLinearAccScale=<0.1 0.1 0.1>
    componentLinearVelScale=<0.2 0.2 0.2>
    simSpaceSettings {
     maxAngularAcceleration=3
     maxAngularVelocity=5
     maxLinearAcceleration=500
     maxLinearVelocity=1500
     velocityScaleZ=1
     externalLinearDragV=<1 1 0>
    }
   }
   onGround_Standing {
    componentLinearAccScale=<0.2 0.2 0.3>
    componentLinearVelScale=<0.3 0.3 0.4>
    simSpaceSettings {
     maxAngularAcceleration=50
     maxAngularVelocity=5
     maxLinearAcceleration=500
     maxLinearVelocity=1500
     velocityScaleZ=1
     externalLinearDragV=<1 1 0>
    }
   }
   rail_Grinding {simSpaceSettings {
     maxAngularAcceleration=3
     maxAngularVelocity=5
     maxLinearAcceleration=500
     maxLinearVelocity=1500
     velocityScaleZ=1
     worldAlpha=1
     externalLinearDragV=<1 1 0>
    }}
   riding_Moving {simSpaceSettings {
     maxAngularAcceleration=3
     maxAngularVelocity=5
     maxLinearAcceleration=500
     maxLinearVelocity=1500
     velocityScaleZ=1
     externalLinearDragV=<1 1 0>
    }}
   riding_Moving_Targeting {simSpaceSettings {
     maxAngularAcceleration=3
     maxAngularVelocity=5
     maxLinearAcceleration=500
     maxLinearVelocity=1500
     velocityScaleZ=1
     externalLinearDragV=<1 1 0>
    }}
   riding_Standing {simSpaceSettings {
     maxAngularAcceleration=3
     maxAngularVelocity=5
     maxLinearAcceleration=500
     maxLinearVelocity=1500
     velocityScaleZ=1
     externalLinearDragV=<1 1 0>
    }}
   skydive_DivingDown {
    componentLinearAccScale=<0 0 0>
    componentLinearVelScale=<0 0 0>
    simSpaceSettings {
     maxAngularAcceleration=3
     maxAngularVelocity=5
     maxLinearAcceleration=500
     maxLinearVelocity=1500
     velocityScaleZ=1
     externalLinearDragV=<1 1 0>
    }
   }
   skydive_FloatingDown {
    componentLinearAccScale=<0 0 0>
    componentLinearVelScale=<0 0 0>
    simSpaceSettings {
     maxAngularAcceleration=3
     maxAngularVelocity=5
     maxLinearAcceleration=500
     maxLinearVelocity=1500
     velocityScaleZ=1
     externalLinearDragV=<1 1 0>
    }
   }
   skydive_GoingUp {simSpaceSettings {
     maxAngularAcceleration=3
     maxAngularVelocity=5
     maxLinearAcceleration=500
     maxLinearVelocity=1500
     velocityScaleZ=1
     externalLinearDragV=<1 1 0>
    }}
   skydive_Parachuting {
    componentLinearAccScale=<0 0 0>
    componentLinearVelScale=<0 0 0>
    simSpaceSettings {
     maxAngularAcceleration=3
     maxAngularVelocity=5
     maxLinearAcceleration=500
     maxLinearVelocity=1500
     velocityScaleZ=1
     externalLinearDragV=<1 1 0>
    }
   }
   skydive_Surfing {simSpaceSettings {
     maxAngularAcceleration=3
     maxAngularVelocity=5
     maxLinearAcceleration=500
     maxLinearVelocity=1500
     velocityScaleZ=1
     externalLinearDragV=<1 1 0>
    }}
   swimming {simSpaceSettings {
     maxAngularAcceleration=3
     maxAngularVelocity=5
     maxLinearAcceleration=500
     maxLinearVelocity=1500
     velocityScaleZ=1
     externalLinearDragV=<1 1 0>
    }}
   tactical_Sprint {
    componentLinearAccScale=<0.2 0.2 0.2>
    componentLinearVelScale=<0.2 0.2 0.2>
    simSpaceSettings {
     maxAngularAcceleration=3
     maxAngularVelocity=5
     maxLinearAcceleration=500
     maxLinearVelocity=1500
     velocityScaleZ=1
     externalLinearDragV=<1 1 0>
    }
   }
   ziplining {simSpaceSettings {
     maxAngularAcceleration=0.5
     maxAngularVelocity=1
     maxLinearAcceleration=50
     maxLinearVelocity=200
     velocityScaleZ=1
     externalLinearDragV=<0.5 0.5 0>
    }}
  }
 }
 RBAN Settings Out {
  type=RBANSettingStruct
  default {
   componentLinearAccScale=<0 0 0>
   componentLinearVelScale=<0 0 0>
   simSpaceSettings {
    dampingAlpha=1
    maxAngularAcceleration=1000
    maxAngularVelocity=3
    maxLinearAcceleration=300
    maxLinearVelocity=700
    velocityScaleZ=0
    worldAlpha=1
    externalAngularVelocity=<0 0 0>
    externalLinearDragV=<0 0.5 0>
    externalLinearVelocity=<0 0 0>
   }
  }
 }
 reposeCapeTrail {type=boolean}
 Sim Space Settings {
  type=SimSpaceSettings
  default {
   dampingAlpha=1
   maxAngularAcceleration=40
   maxAngularVelocity=5
   maxLinearAcceleration=200
   maxLinearVelocity=200
   velocityScaleZ=1
   worldAlpha=1
   externalLinearDragV=<0.4 0.4 0.4>
  }
 }
 Sim Space Settings_Default {
  type=SimSpaceSettings
  default {
   dampingAlpha=1
   maxAngularAcceleration=40
   maxAngularVelocity=5
   maxLinearAcceleration=200
   maxLinearVelocity=200
   velocityScaleZ=1
   worldAlpha=1
   externalLinearDragV=<0.4 0.4 0.4>
  }
 }
 Sim Space Settings_FE {
  type=SimSpaceSettings
  default {
   dampingAlpha=1
   maxAngularAcceleration=40
   maxAngularVelocity=5
   maxLinearAcceleration=100
   maxLinearVelocity=100
   velocityScaleZ=0
   worldAlpha=0.5
   externalLinearDragV=<0.4 0.4 0.4>
  }
 }
 Simulation Space {type=AnimPhysSimSpaceType, default=RootRelative}
 Skel_Mesh {type=SkeletalMeshComponent}
 skydive {type=boolean}
 skydiveNoParachute {type=boolean}
 TrailControllerRot {type=Rotator}
 Wind Gust Params {
  type=WindGustParamsStruct
  default {
   dBNO {
    maxDownInterpSpeed=5
    maxGustDuration=2
    maxGustStrength=7
    maxTimeBetweenGusts=7
    maxUpInterpSpeed=5
    minDownInterpSpeed=2
    minGustDuration=0.5
    minGustStrength=3
    minTimeBetweenGusts=0
    minUpInterpSpeed=2
    windTurbulenceScalarA=0.1
    windTurbulenceScalarB=1
    windAmplitude=<0 0 0>
    windFrequency=<0 0 0>
    windOffset=<0 0 0>
   }
   dBNO_Carried {
    maxDownInterpSpeed=5
    maxGustDuration=2
    maxGustStrength=7
    maxTimeBetweenGusts=7
    maxUpInterpSpeed=5
    minDownInterpSpeed=2
    minGustDuration=0.5
    minGustStrength=3
    minTimeBetweenGusts=0
    minUpInterpSpeed=2
    windTurbulenceScalarA=0.1
    windTurbulenceScalarB=1
    windAmplitude=<0 0 0>
    windFrequency=<0 0 0>
    windOffset=<0 0 0>
   }
   dBNO_Impulse {
    maxDownInterpSpeed=5
    maxGustDuration=2
    maxGustStrength=7
    maxTimeBetweenGusts=7
    maxUpInterpSpeed=5
    minDownInterpSpeed=2
    minGustDuration=0.5
    minGustStrength=3
    minTimeBetweenGusts=0
    minUpInterpSpeed=2
    windTurbulenceScalarA=0.1
    windTurbulenceScalarB=1
    windAmplitude=<0 0 0>
    windFrequency=<0 0 0>
    windOffset=<0 0 0>
   }
   dBNO_RollBwd {
    maxDownInterpSpeed=5
    maxGustDuration=2
    maxGustStrength=7
    maxTimeBetweenGusts=7
    maxUpInterpSpeed=5
    minDownInterpSpeed=2
    minGustDuration=0.5
    minGustStrength=3
    minTimeBetweenGusts=0
    minUpInterpSpeed=2
    windTurbulenceScalarA=0.1
    windTurbulenceScalarB=1
    windAmplitude=<0 0 0>
    windFrequency=<0 0 0>
    windOffset=<0 0 0>
   }
   dBNO_RollFwd {
    maxDownInterpSpeed=5
    maxGustDuration=2
    maxGustStrength=7
    maxTimeBetweenGusts=7
    maxUpInterpSpeed=5
    minDownInterpSpeed=2
    minGustDuration=0.5
    minGustStrength=3
    minTimeBetweenGusts=0
    minUpInterpSpeed=2
    windTurbulenceScalarA=0.1
    windTurbulenceScalarB=1
    windAmplitude=<0 0 0>
    windFrequency=<0 0 0>
    windOffset=<0 0 0>
   }
   dBNO_Tumbling {
    maxDownInterpSpeed=5
    maxGustDuration=2
    maxGustStrength=7
    maxTimeBetweenGusts=7
    maxUpInterpSpeed=5
    minDownInterpSpeed=2
    minGustDuration=0.5
    minGustStrength=3
    minTimeBetweenGusts=0
    minUpInterpSpeed=2
    windTurbulenceScalarA=0.1
    windTurbulenceScalarB=1
    windAmplitude=<0 0 0>
    windFrequency=<0 0 0>
    windOffset=<0 0 0>
   }
   emoteOrMelee {
    maxDownInterpSpeed=5
    maxGustDuration=2
    maxGustStrength=7
    maxTimeBetweenGusts=7
    maxUpInterpSpeed=5
    minDownInterpSpeed=2
    minGustDuration=0.5
    minGustStrength=3
    minTimeBetweenGusts=0
    minUpInterpSpeed=2
    windTurbulenceScalarA=0.1
    windTurbulenceScalarB=1
    windAmplitude=<30 25 50>
    windFrequency=<2 3 5>
    windOffset=<-50 -10 0>
   }
   front_End {
    maxDownInterpSpeed=5
    maxGustDuration=8
    maxGustStrength=7
    maxTimeBetweenGusts=8
    maxUpInterpSpeed=5
    minDownInterpSpeed=2
    minGustDuration=0.2
    minGustStrength=2
    minTimeBetweenGusts=1
    minUpInterpSpeed=2
    windTurbulenceScalarA=0.1
    windTurbulenceScalarB=1
    windAmplitude=<0 0 0>
    windFrequency=<1.5 2.3 3.7>
    windOffset=<0 0 0>
   }
   onGround_Crouch_Moving {
    maxDownInterpSpeed=5
    maxGustDuration=2
    maxGustStrength=7
    maxTimeBetweenGusts=7
    maxUpInterpSpeed=5
    minDownInterpSpeed=2
    minGustDuration=0.5
    minGustStrength=3
    minTimeBetweenGusts=0
    minUpInterpSpeed=2
    windTurbulenceScalarA=0.1
    windTurbulenceScalarB=1
    windAmplitude=<100 25 75>
    windFrequency=<1.5 0.5 7.5>
    windOffset=<100 -100 -175>
   }
   onGround_Falling {
    maxDownInterpSpeed=5
    maxGustDuration=2
    maxGustStrength=7
    maxTimeBetweenGusts=7
    maxUpInterpSpeed=5
    minDownInterpSpeed=2
    minGustDuration=0.5
    minGustStrength=3
    minTimeBetweenGusts=0
    minUpInterpSpeed=2
    windTurbulenceScalarA=0.1
    windTurbulenceScalarB=1
    windAmplitude=<0 0 0>
    windFrequency=<0 0 0>
    windOffset=<0 0 0>
   }
   onGround_Moving {
    maxDownInterpSpeed=5
    maxGustDuration=8
    maxGustStrength=1
    maxTimeBetweenGusts=8
    maxUpInterpSpeed=5
    minDownInterpSpeed=2
    minGustDuration=5
    minGustStrength=0
    minTimeBetweenGusts=5
    minUpInterpSpeed=2
    windTurbulenceScalarA=0.2
    windTurbulenceScalarB=1
    windAmplitude=<10 20 20>
    windFrequency=<2 3 2>
    windOffset=<0 -100 -100>
   }
   onGround_Standing {
    maxDownInterpSpeed=5
    maxGustDuration=8
    maxGustStrength=7
    maxTimeBetweenGusts=8
    maxUpInterpSpeed=5
    minDownInterpSpeed=2
    minGustDuration=0.2
    minGustStrength=2
    minTimeBetweenGusts=1
    minUpInterpSpeed=2
    windTurbulenceScalarA=0.1
    windTurbulenceScalarB=1
    windAmplitude=<7 5 24>
    windFrequency=<1.5 2.3 3.7>
    windOffset=<-80 0 -70>
   }
   rail_Grinding {
    maxDownInterpSpeed=5
    maxGustDuration=8
    maxGustStrength=7
    maxTimeBetweenGusts=8
    maxUpInterpSpeed=5
    minDownInterpSpeed=2
    minGustDuration=0.2
    minGustStrength=3
    minTimeBetweenGusts=1
    minUpInterpSpeed=2
    windTurbulenceScalarA=0.1
    windTurbulenceScalarB=0.8
    windAmplitude=<20 65 42>
    windFrequency=<5 6 8.6>
    windOffset=<50 -500 -400>
   }
   riding_Moving {
    maxDownInterpSpeed=5
    maxGustDuration=2
    maxGustStrength=7
    maxTimeBetweenGusts=7
    maxUpInterpSpeed=5
    minDownInterpSpeed=2
    minGustDuration=0.5
    minGustStrength=3
    minTimeBetweenGusts=0
    minUpInterpSpeed=2
    windTurbulenceScalarA=0.1
    windTurbulenceScalarB=1
    windAmplitude=<0 0 0>
    windFrequency=<0 0 0>
    windOffset=<0 0 0>
   }
   riding_Moving_Targeting {
    maxDownInterpSpeed=5
    maxGustDuration=2
    maxGustStrength=7
    maxTimeBetweenGusts=7
    maxUpInterpSpeed=5
    minDownInterpSpeed=2
    minGustDuration=0.5
    minGustStrength=3
    minTimeBetweenGusts=0
    minUpInterpSpeed=2
    windTurbulenceScalarA=0.1
    windTurbulenceScalarB=1
    windAmplitude=<0 0 0>
    windFrequency=<0 0 0>
    windOffset=<0 0 0>
   }
   riding_Standing {
    maxDownInterpSpeed=5
    maxGustDuration=2
    maxGustStrength=7
    maxTimeBetweenGusts=7
    maxUpInterpSpeed=5
    minDownInterpSpeed=2
    minGustDuration=0.5
    minGustStrength=3
    minTimeBetweenGusts=0
    minUpInterpSpeed=2
    windTurbulenceScalarA=0.1
    windTurbulenceScalarB=1
    windAmplitude=<0 0 0>
    windFrequency=<0 0 0>
    windOffset=<0 0 0>
   }
   skydive_DivingDown {
    maxDownInterpSpeed=5
    maxGustDuration=8
    maxGustStrength=7
    maxTimeBetweenGusts=8
    maxUpInterpSpeed=5
    minDownInterpSpeed=2
    minGustDuration=0.2
    minGustStrength=3
    minTimeBetweenGusts=1
    minUpInterpSpeed=2
    windTurbulenceScalarA=0.1
    windTurbulenceScalarB=4
    windAmplitude=<20 25 150>
    windFrequency=<2 4 30>
    windOffset=<-50 -50 0>
   }
   skydive_FloatingDown {
    maxDownInterpSpeed=5
    maxGustDuration=8
    maxGustStrength=1
    maxTimeBetweenGusts=8
    maxUpInterpSpeed=5
    minDownInterpSpeed=2
    minGustDuration=0.2
    minGustStrength=0
    minTimeBetweenGusts=1
    minUpInterpSpeed=2
    windTurbulenceScalarA=0.5
    windTurbulenceScalarB=1.5
    windAmplitude=<70 20 50>
    windFrequency=<10 2 3>
    windOffset=<0 -250 100>
   }
   skydive_GoingUp {
    maxDownInterpSpeed=5
    maxGustDuration=8
    maxGustStrength=7
    maxTimeBetweenGusts=8
    maxUpInterpSpeed=5
    minDownInterpSpeed=2
    minGustDuration=0.2
    minGustStrength=3
    minTimeBetweenGusts=1
    minUpInterpSpeed=2
    windTurbulenceScalarA=0.1
    windTurbulenceScalarB=0.8
    windAmplitude=<50 20 35>
    windFrequency=<8 2 3>
    windOffset=<0 -500 25>
   }
   skydive_Parachuting {
    maxDownInterpSpeed=5
    maxGustDuration=8
    maxGustStrength=1
    maxTimeBetweenGusts=8
    maxUpInterpSpeed=5
    minDownInterpSpeed=2
    minGustDuration=5
    minGustStrength=0
    minTimeBetweenGusts=5
    minUpInterpSpeed=2
    windTurbulenceScalarA=0.2
    windTurbulenceScalarB=1
    windAmplitude=<50 50 75>
    windFrequency=<15 6 2>
    windOffset=<0 -200 -200>
   }
   skydive_Surfing {
    maxDownInterpSpeed=5
    maxGustDuration=8
    maxGustStrength=1
    maxTimeBetweenGusts=8
    maxUpInterpSpeed=5
    minDownInterpSpeed=2
    minGustDuration=5
    minGustStrength=0
    minTimeBetweenGusts=5
    minUpInterpSpeed=2
    windTurbulenceScalarA=0.2
    windTurbulenceScalarB=1
    windAmplitude=<50 50 75>
    windFrequency=<15 6 2>
    windOffset=<0 -250 -200>
   }
   swimming {
    maxDownInterpSpeed=5
    maxGustDuration=2
    maxGustStrength=7
    maxTimeBetweenGusts=7
    maxUpInterpSpeed=5
    minDownInterpSpeed=2
    minGustDuration=0.5
    minGustStrength=3
    minTimeBetweenGusts=0
    minUpInterpSpeed=2
    windTurbulenceScalarA=0.1
    windTurbulenceScalarB=1
    windAmplitude=<10 25 20>
    windFrequency=<1.5 0.5 7.5>
    windOffset=<0 -150 50>
   }
   tactical_Sprint {
    maxDownInterpSpeed=5
    maxGustDuration=8
    maxGustStrength=1
    maxTimeBetweenGusts=8
    maxUpInterpSpeed=5
    minDownInterpSpeed=2
    minGustDuration=5
    minGustStrength=0
    minTimeBetweenGusts=5
    minUpInterpSpeed=2
    windTurbulenceScalarA=0.2
    windTurbulenceScalarB=1
    windAmplitude=<10 50 50>
    windFrequency=<15 6 2>
    windOffset=<0 -100 120>
   }
   ziplining {
    maxDownInterpSpeed=5
    maxGustDuration=8
    maxGustStrength=1
    maxTimeBetweenGusts=8
    maxUpInterpSpeed=5
    minDownInterpSpeed=2
    minGustDuration=5
    minGustStrength=0
    minTimeBetweenGusts=5
    minUpInterpSpeed=2
    windTurbulenceScalarA=0.2
    windTurbulenceScalarB=1
    windAmplitude=<50 50 75>
    windFrequency=<15 6 2>
    windOffset=<0 -200 -200>
   }
  }
 }
 Z Movement {type=number}
]
