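# Control Rig for the BinGrass female character's face accessory part that dampens neck and head rotation.
# The forward solve loops over neck_01, neck_02 and head, reading each bone's current local transform.
# It converts rotation to ZXY Euler angles, halves Y and Z while keeping X intact, then writes it back with translation and scale unchanged.
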ControlRigBlueprint /BRCosmetics/Characters/Player/Female/Medium/Bodies/F_MED_BinGrass/Meshes/Parts/F_MED_BinGrass_FaceAcc_CtrlRig
graphs [
 RigVMFunctionLibrary {links [], nodes []}
 RigVMModel {nodes [
   RigUnit_BeginExecution = &RigUnit_BeginExecution()
   DISPATCH_RigVMDispatch_ArrayIterator = RigVMDispatchNode(
    Array=&RigUnit_ItemArray(Items=((Type=Bone,Name="neck_01"),(Type=Bone,Name="neck_02"),(Type=Bone,Name="head"))).Items
    ExecuteContext=@RigUnit_BeginExecution.ExecutePin
   ) {template="DISPATCH_RigVMDispatch_ArrayIterator(in Array,out Element,out Index,out Count,out Ratio)"}
   RigUnit_GetTransform = &RigUnit_GetTransform(Space=LocalSpace, bInitial=false, Item=@DISPATCH_RigVMDispatch_ArrayIterator.Element)
   RigVMFunction_MathQuaternionToEuler = &RigVMFunction_MathQuaternionToEuler(RotationOrder=ZXY, Value=@RigUnit_GetTransform.Transform.Rotation)
   Multiply = &RigVMFunction_MathDoubleMul(B=0.5, A=@RigVMFunction_MathQuaternionToEuler.Result.Z)
   RigVMFunction_MathQuaternionFromEuler = &RigVMFunction_MathQuaternionFromEuler(
    Euler=(X=0,Y=0,Z=0)
    RotationOrder=ZXY
    Euler.Z=@Multiply.Result
    Euler.Y=&RigVMFunction_MathDoubleMul(B=0.5, A=@RigVMFunction_MathQuaternionToEuler.Result.Y).Result
    Euler.X=&RigVMFunction_MathDoubleMul(B=1, A=@RigVMFunction_MathQuaternionToEuler.Result.X).Result
   )
   > RigUnit_SetTransform = &RigUnit_SetTransform(
    Space=LocalSpace
    Value=(Rotation=(X=-0.0029,Y=0.7071,Z=0.0029,W=-0.7071),Translation=(X=0,Y=-1.6873,Z=152.0342),Scale3D=(X=1,Y=1,Z=1))
    Weight=1
    bInitial=False
    Item=@DISPATCH_RigVMDispatch_ArrayIterator.Element
    Value.Rotation=@RigVMFunction_MathQuaternionFromEuler.Result
    Value.Translation=@RigUnit_GetTransform.Transform.Translation
    Value.Scale3D=@RigUnit_GetTransform.Transform.Scale3D
   )
  ]}
]
hierarchy [
 root {loc=<0 0 0>, rot=<0 0 0 1>}
 pelvis {parent=root, loc=<0 -0.5301 98.0571>, rot=<0.0127 0.707 -0.0127 -0.707>}
 spine_01 {parent=pelvis, loc=<6.1208 -0.9328 0>, rot=<0 0 0.034 -0.9994>}
 spine_02 {parent=spine_01, loc=<6.5997 -0.2095 0>, rot=<0 0 0.0238 -0.9997>}
 spine_03 {parent=spine_02, loc=<6.4914 -0.1237 0>, rot=<0 0 0.0149 -0.9999>}
 spine_04 {parent=spine_03, loc=<6.9598 0.0794 0>, rot=<0 0 -0.0132 -0.9999>}
 spine_05 {parent=spine_04, loc=<6.9987 0 0>, rot=<0 0 -0.0015 -1>}
 neck_01 {parent=spine_05, loc=<11.3138 0.1529 0>, rot=<0 0 -0.2003 -0.9797>}
 neck_02 {parent=neck_01, loc=<4.5627 0 0>, rot=<0 0 0.0504 -0.9987>}
 head {parent=neck_02, loc=<5.3914 0 0>, rot=<0 0 0.115 -0.9934>}
 clavicle_l {parent=spine_05, loc=<5.4831 -0.1053 -3.2946>, rot=<0.0465 -0.7289 -0.009 -0.6829>}
 clavicle_r {parent=spine_05, loc=<5.4831 -0.1053 3.2946>, rot=<0.7289 0.0465 -0.6829 0.009>}
]
